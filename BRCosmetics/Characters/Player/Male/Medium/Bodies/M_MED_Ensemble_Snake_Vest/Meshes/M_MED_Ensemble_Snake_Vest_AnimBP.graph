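# Animation Blueprint for the male medium Snake Vest ensemble body, adding corrective deformation to glute and front vest bones as the thighs rotate.
# The event graph converts each thigh's transform delta against the pelvis into a per-side alpha every update.
# The AnimGraph copies the attached parent's pose, then applies chained additive bone-space Transform Bone nodes weighted by those alphas.
# A deform Control Rig node is present but unconnected to the output.
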AnimBlueprint /BRCosmetics/Characters/Player/Male/Medium/Bodies/M_MED_Ensemble_Snake_Vest/Meshes/M_MED_Ensemble_Snake_Vest_AnimBP
target_skeleton=&/M_MED_Ensemble_Snake_Vest_Skeleton
graphs [
 AnimGraph {nodes [
   Control Rig() {
    alphaBoolBlend {blendOption=Linear}
    controlRigAssetReference {blueprintRigClass=/Game/Characters/Player/Male/Medium/Bodies/M_MED_Ensemble_Snake_Vest/Meshes/M_MED_EnsembleSnakeVestDeformRig_Master_ControlRig1.M_MED_EnsembleSnakeVestDeformRig_Master_ControlRig1_C}
   }
   n1 = Get thigh_rot_up_l() -> (thigh_rot_up_l:double)
   n2 = "Transform (Modify) Bone - Bone: dyn_glute_l"(
    ComponentPose=Local To Component(LocalPose=Copy Pose From Mesh() {bUseAttachedParent=true})
    Alpha=@n1
    Translation="-12,-10,0"
    Rotation="0,-15,15"
    Scale="1,1,1"
   ) {
    rotationMode=BMM_Additive
    rotationSpace=BCS_BoneSpace
    scaleMode=BMM_Additive
    scaleSpace=BCS_BoneSpace
    translationMode=BMM_Additive
    translationSpace=BCS_BoneSpace
    alphaBoolBlend {blendOption=Linear}
    boneToModify {boneName=dyn_glute_l}
    rotation {roll=15, yaw=-15}
    translation=<-12 -10>
   }
   n3 = "Transform (Modify) Bone - Bone: def_vest_ft_l"(ComponentPose=@n2, Alpha=@n1, Translation="-21,0,18", Rotation="0,-20.000,0", Scale="1,1,1") {
    rotationMode=BMM_Additive
    rotationSpace=BCS_BoneSpace
    scaleMode=BMM_Additive
    scaleSpace=BCS_BoneSpace
    translationMode=BMM_Additive
    translationSpace=BCS_BoneSpace
    alphaBoolBlend {blendOption=Linear}
    boneToModify {boneName=def_vest_ft_l}
    rotation {yaw=-20}
    translation {x=-21, z=18}
   }
   n4 = Get thigh_rot_up_r() -> (thigh_rot_up_r:double)
   n5 = "Transform (Modify) Bone - Bone: dyn_glute_r"(ComponentPose=@n3, Alpha=@n4, Translation="-12,-10,0", Rotation="0,-15,-15", Scale="1,1,1") {
    rotationMode=BMM_Additive
    rotationSpace=BCS_BoneSpace
    scaleMode=BMM_Additive
    scaleSpace=BCS_BoneSpace
    translationMode=BMM_Additive
    translationSpace=BCS_BoneSpace
    alphaBoolBlend {blendOption=Linear}
    boneToModify {boneName=dyn_glute_r}
    rotation {roll=-15, yaw=-15}
    translation=<-12 -10>
   }
   n6 = "Transform (Modify) Bone - Bone: def_vest_ft_r"(ComponentPose=@n5, Alpha=@n4, Translation="-30,0,12", Rotation="50,0,0", Scale="1,1,1") {
    rotationMode=BMM_Additive
    rotationSpace=BCS_BoneSpace
    scaleMode=BMM_Additive
    scaleSpace=BCS_BoneSpace
    translationMode=BMM_Additive
    translationSpace=BCS_BoneSpace
    alphaBoolBlend {blendOption=Linear}
    boneToModify {boneName=def_vest_ft_r}
    rotation {pitch=50}
    translation {x=-30, z=12}
   }
   Output Pose(Result=Component To Local(ComponentPose=@n6)) {layerGroup=None}
  ]}
 EventGraph {nodes [
   Event BlueprintUpdateAnimation()
   n1 = GetOwnerSkeletalMeshComponent()
   > Set thigh_rot_up_r(thigh_rot_up_r:double=GetAlphaFromReferenceTransformDelta(SkeletalMeshComponent=@n1, BoneName=thigh_r, BaseName=pelvis, InputRange=(X=-15,Y=-115), TwistAxis="1,0,0")) -> (Output_Get:double)
   > Set thigh_rot_up_l(thigh_rot_up_l:double=GetAlphaFromReferenceTransformDelta(SkeletalMeshComponent=@n1, BoneName=thigh_l, BaseName=pelvis, InputRange=(X=-10,Y=-110), TwistAxis="1,0,0")) -> (Output_Get:double)
  ]}
]
variables [
 thigh_rot_up_l {type=number}
 thigh_rot_up_r {type=number}
]
