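# Physics asset for the dynamic front hair of the male medium Snake Vest cosmetic character part.
# Left and right three-segment capsule chains hang from the neck through offset, MSTR and AIM bones, with tight twist and moderate swing limits, slightly wider on the right.
# Larger capsules on the upper spine, neck, clavicles and upper arms form the body volume around the hair, linked by angularly free constraints.
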PhysicsAsset /BRCosmetics/Characters/Player/Male/Medium/Bodies/M_MED_Ensemble_Snake_Vest/Meshes/Parts/M_MED_Ensemble_Snake_Vest_FaceAcc_Dyn_Front_Hair
bodies {
 clavicle_l [{
   shapeType=Capsule
   length=5.25
   radius=7.5
   shapeName=clavicle_l_capsule
   center=<8.4521 -1.2518 -2.2519>
   rotation=<rot 86.2624 -0.0144 -5.0145>
  }]
 clavicle_r [{
   shapeType=Capsule
   length=3.125
   radius=8.125
   shapeName=clavicle_l_capsule
   center=<-9.2473 1.0326 0.8503>
   rotation=<rot 76.2729 3.6537 178.625>
  }]
 dyn_hair_ft_l_01 [{shapeType=Capsule, length=5.125, radius=2.625, shapeName=dyn_hair_r_01_capsule, center=<5 -1 0>, rotation=<rot 90 -177.5355 172.4647>}]
 dyn_hair_ft_l_02 [{shapeType=Capsule, length=5.875, radius=2.125, shapeName=dyn_hair_r_02_capsule, center=<5 0 0>, rotation=<rot 90 73.044 68.0432>}]
 dyn_hair_ft_l_03 [{shapeType=Capsule, length=5.5, radius=1.375, shapeName=dyn_hair_r_03_capsule, center=<5 0 0>, rotation=<rot 90 166.7373 166.737>}]
 dyn_hair_ft_l_AIM [{shapeType=Capsule, length=0.505, radius=0.505, shapeName=dyn_hair_l_MSTR_capsule, rotation=<rot 0 -90 0>}]
 dyn_hair_ft_l_MSTR [{shapeType=Capsule, length=0.505, radius=0.505, shapeName=dyn_hair_l_MSTR_capsule, rotation=<rot 0 -90 0>}]
 dyn_hair_ft_r_01 [{shapeType=Capsule, length=7.125, radius=2.25, shapeName=dyn_hair_r_01_capsule, center=<5 2 0>, rotation=<rot 90 177.5355 -172.4641>}]
 dyn_hair_ft_r_02 [{shapeType=Capsule, length=8.25, radius=2, shapeName=dyn_hair_r_02_capsule, center=<5 0 0>, rotation=<rot 90 -77.1723 -72.172>}]
 dyn_hair_ft_r_03 [{shapeType=Capsule, length=5.125, radius=1.375, shapeName=dyn_hair_r_03_capsule, center=<5 0 0>, rotation=<rot 90 -78.2837 -83.2836>}]
 dyn_hair_ft_r_AIM [{shapeType=Capsule, length=0.505, radius=0.505, shapeName=dyn_hair_l_MSTR_capsule, rotation=<rot 0 -90 0>}]
 dyn_hair_ft_r_MSTR [{shapeType=Capsule, length=0.505, radius=0.505, shapeName=dyn_hair_l_MSTR_capsule, rotation=<rot 0 -90 0>}]
 dyn_hair_offset [{shapeType=Capsule, length=0.505, radius=0.505, shapeName=dyn_hair_r_MSTR_capsule, rotation=<rot 0 -90 0>}]
 neck_01 [{
   shapeType=Capsule
   length=3.6965
   radius=7.0617
   shapeName=head_capsule
   center=<3.0345 2.5883 -0.2642>
   rotation=<rot 90 -165.9638 -155.9632>
  }]
 spine_03 [{shapeType=Capsule, length=9.375, radius=12.375, shapeName=spine_05_capsule, center=<8.0528 1.3592 0>}]
 spine_04 [{shapeType=Capsule, length=14.875, radius=11.625, shapeName=spine_05_capsule, center=<7.794 0.4016 0>}]
 spine_05 [
  {shapeType=Capsule, length=18.5, radius=9.375, shapeName=spine_05_capsule, center=<3.5089 -0.3109 0>}
  {shapeType=Capsule, length=11.875, radius=8.5, shapeName=spine_05_capsule, center=<6.6936 0.0344 0>}
 ]
 upperarm_twist_01_l [{
   shapeType=Capsule
   length=16.125
   radius=10.75
   shapeName=clavicle_l_capsule
   center=<8.4143 -8.6709 0.1001>
   rotation=<rot 89.5647 92.4923 92.5104>
  }]
 upperarm_twist_01_r [{
   shapeType=Capsule
   length=16.125
   radius=10.75
   shapeName=clavicle_l_capsule
   center=<-8.4143 8.6709 -0.1001>
   rotation=<rot 89.5647 92.4923 -87.4896>
  }]
}
constraints [
 {
  bone1Name=spine_05
  bone2Name=spine_04
  swing1LimitDegrees=45
  swing1Motion=Free
  swing2LimitDegrees=45
  swing2Motion=Free
  twistLimitDegrees=45
  twistMotion=Free
 }
 {
  bone1Name=neck_01
  bone2Name=spine_05
  swing1LimitDegrees=45
  swing1Motion=Free
  swing2LimitDegrees=45
  swing2Motion=Free
  twistLimitDegrees=45
  twistMotion=Free
 }
 {
  bone1Name=spine_04
  bone2Name=spine_03
  swing1LimitDegrees=45
  swing1Motion=Free
  swing2LimitDegrees=45
  swing2Motion=Free
  twistLimitDegrees=45
  twistMotion=Free
 }
 {
  bone1Name=dyn_hair_ft_r_MSTR
  bone2Name=dyn_hair_offset
  swing1LimitDegrees=45
  swing1Motion=Free
  swing2LimitDegrees=45
  swing2Motion=Free
  twistLimitDegrees=45
  twistMotion=Free
 }
 {bone1Name=dyn_hair_ft_r_01, bone2Name=dyn_hair_ft_r_AIM, swing1LimitDegrees=30, swing2LimitDegrees=20, twistLimitDegrees=2}
 {bone1Name=dyn_hair_ft_r_02, bone2Name=dyn_hair_ft_r_01, swing1LimitDegrees=30, swing2LimitDegrees=20, twistLimitDegrees=2}
 {bone1Name=dyn_hair_ft_r_03, bone2Name=dyn_hair_ft_r_02, swing1LimitDegrees=30, swing2LimitDegrees=20, twistLimitDegrees=2}
 {
  bone1Name=dyn_hair_ft_l_MSTR
  bone2Name=dyn_hair_offset
  swing1LimitDegrees=45
  swing1Motion=Free
  swing2LimitDegrees=45
  swing2Motion=Free
  twistLimitDegrees=45
  twistMotion=Free
 }
 {bone1Name=dyn_hair_ft_l_01, bone2Name=dyn_hair_ft_l_AIM, swing1LimitDegrees=20, swing2LimitDegrees=20, twistLimitDegrees=2}
 {bone1Name=dyn_hair_ft_l_02, bone2Name=dyn_hair_ft_l_01, swing1LimitDegrees=20, swing2LimitDegrees=20, twistLimitDegrees=2}
 {bone1Name=dyn_hair_ft_l_03, bone2Name=dyn_hair_ft_l_02, swing1LimitDegrees=20, swing2LimitDegrees=20, twistLimitDegrees=2}
 {
  bone1Name=clavicle_l
  bone2Name=spine_05
  swing1LimitDegrees=45
  swing1Motion=Free
  swing2LimitDegrees=45
  swing2Motion=Free
  twistLimitDegrees=45
  twistMotion=Free
 }
 {
  bone1Name=clavicle_r
  bone2Name=spine_05
  swing1LimitDegrees=45
  swing1Motion=Free
  swing2LimitDegrees=45
  swing2Motion=Free
  twistLimitDegrees=45
  twistMotion=Free
 }
 {
  bone1Name=dyn_hair_ft_r_AIM
  bone2Name=dyn_hair_ft_r_MSTR
  swing1LimitDegrees=45
  swing1Motion=Free
  swing2LimitDegrees=45
  swing2Motion=Free
  twistLimitDegrees=45
  twistMotion=Free
 }
 {
  bone1Name=dyn_hair_ft_l_AIM
  bone2Name=dyn_hair_ft_l_MSTR
  swing1LimitDegrees=45
  swing1Motion=Free
  swing2LimitDegrees=45
  swing2Motion=Free
  twistLimitDegrees=45
  twistMotion=Free
 }
 {
  bone1Name=dyn_hair_offset
  bone2Name=neck_01
  swing1LimitDegrees=45
  swing1Motion=Free
  swing2LimitDegrees=45
  swing2Motion=Free
  twistLimitDegrees=45
  twistMotion=Free
 }
 {
  bone1Name=upperarm_twist_01_r
  bone2Name=clavicle_r
  swing1LimitDegrees=45
  swing1Motion=Free
  swing2LimitDegrees=45
  swing2Motion=Free
  twistLimitDegrees=45
  twistMotion=Free
 }
 {
  bone1Name=upperarm_twist_01_l
  bone2Name=clavicle_l
  swing1LimitDegrees=45
  swing1Motion=Free
  swing2LimitDegrees=45
  swing2Motion=Free
  twistLimitDegrees=45
  twistMotion=Free
 }
]
