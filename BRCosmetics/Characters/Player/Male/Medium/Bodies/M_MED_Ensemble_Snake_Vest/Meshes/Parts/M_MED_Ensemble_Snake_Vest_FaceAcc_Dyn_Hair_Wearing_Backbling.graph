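# Physics asset for the dynamic hair of the male medium Snake Vest character's face accessory, in the variant named for wearing a backbling.
# Capsules cover pelvis, spine, neck, clavicles and upper arms, alongside three back hair chains and two single-segment side pieces.
# Each hair piece hangs from neck_01 through a shared offset bone and tiny MSTR and AIM bodies with free angular constraints.
# Hair segments use limited constraints, with swing tightening toward the back chains' tips.
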PhysicsAsset /BRCosmetics/Characters/Player/Male/Medium/Bodies/M_MED_Ensemble_Snake_Vest/Meshes/Parts/M_MED_Ensemble_Snake_Vest_FaceAcc_Dyn_Hair_Wearing_Backbling
bodies {
 clavicle_l [{
   shapeType=Capsule
   length=3.375
   radius=8.375
   shapeName=clavicle_l_capsule
   center=<9.2473 -1.0326 -0.8503>
   rotation=<rot 76.2729 3.6537 -1.375>
  }]
 clavicle_r [{
   shapeType=Capsule
   length=3.125
   radius=8.125
   shapeName=clavicle_l_capsule
   center=<-9.2473 1.0326 0.8503>
   rotation=<rot 76.2729 3.6537 178.625>
  }]
 dyn_hair_bk_01 [{shapeType=Capsule, length=2.5, radius=5.5, shapeName=dyn_hair_bk_01_capsule, center=<6 0 0>, rotation=<rot 0 0 90>}]
 dyn_hair_bk_02 [{shapeType=Capsule, length=4.125, radius=5.75, shapeName=dyn_hair_bk_02_capsule, center=<5 0 0>, rotation=<rot 90 0 0>}]
 dyn_hair_bk_03 [{shapeType=Capsule, length=4.25, radius=4.5, shapeName=dyn_hair_bk_03_capsule, center=<5 0 0>, rotation=<rot 90 0 0>}]
 dyn_hair_bk_04 [{shapeType=Capsule, length=4.5, radius=4.125, shapeName=dyn_hair_bk_04_capsule, center=<5 0 0>, rotation=<rot 90 0 0>}]
 dyn_hair_bk_05 [{shapeType=Capsule, length=4.5, radius=4.125, shapeName=dyn_hair_bk_04_capsule, center=<5 0 0>, rotation=<rot 90 0 0>}]
 dyn_hair_bk_06 [{shapeType=Capsule, length=5.5, radius=3.125, shapeName=dyn_hair_bk_04_capsule, center=<5 0 0>, rotation=<rot 85.0001 -180 180>}]
 dyn_hair_bk_AIM [{shapeType=Capsule, length=0.505, radius=0.505, shapeName=dyn_hair_bk_MSTR_capsule, rotation=<rot 0 -90 0>}]
 dyn_hair_bk_MSTR [{shapeType=Capsule, length=0.505, radius=0.505, shapeName=dyn_hair_bk_MSTR_capsule, rotation=<rot 0 -90 0>}]
 dyn_hair_bk_l_01 [{shapeType=Capsule, length=4.5, radius=4.875, shapeName=dyn_hair_r_01_capsule, center=<5 0 -1>, rotation=<rot 85.0001 180 180>}]
 dyn_hair_bk_l_02 [{shapeType=Capsule, length=6.375, radius=5, shapeName=dyn_hair_r_02_capsule, center=<6 0 0>, rotation=<rot 80.0002 180 180>}]
 dyn_hair_bk_l_03 [{shapeType=Capsule, length=6.125, radius=4, shapeName=dyn_hair_r_03_capsule, center=<6 0 0>, rotation=<rot 85.0002 -180 180>}]
 dyn_hair_bk_l_04 [{shapeType=Capsule, length=4.875, radius=3.75, shapeName=dyn_hair_r_04_capsule, center=<6 0 0>, rotation=<rot 90 0 0>}]
 dyn_hair_bk_l_AIM [{shapeType=Capsule, length=0.505, radius=0.505, shapeName=dyn_hair_r_MSTR_capsule, rotation=<rot 0 -90 0>}]
 dyn_hair_bk_l_MSTR [{shapeType=Capsule, length=0.505, radius=0.505, shapeName=dyn_hair_r_MSTR_capsule, rotation=<rot 0 -90 0>}]
 dyn_hair_bk_r_01 [{shapeType=Capsule, length=3.625, radius=5.125, shapeName=dyn_hair_r_01_capsule, center=<7.3838 0 -0.7914>, rotation=<rot 85.0001 180 180>}]
 dyn_hair_bk_r_02 [{shapeType=Capsule, length=7.5, radius=5, shapeName=dyn_hair_r_02_capsule, center=<6 0 0>, rotation=<rot 85.0001 -180 -180>}]
 dyn_hair_bk_r_03 [{shapeType=Capsule, length=5.75, radius=3.5, shapeName=dyn_hair_r_03_capsule, center=<6 0 0>, rotation=<rot 90 0 0>}]
 dyn_hair_bk_r_04 [{shapeType=Capsule, length=5.375, radius=3.25, shapeName=dyn_hair_r_04_capsule, center=<6 0 0>, rotation=<rot 90 0 0>}]
 dyn_hair_bk_r_05 [{shapeType=Capsule, length=5.375, radius=3.25, shapeName=dyn_hair_r_04_capsule, center=<6 0 0>, rotation=<rot 90 0 0>}]
 dyn_hair_bk_r_AIM [{shapeType=Capsule, length=0.505, radius=0.505, shapeName=dyn_hair_r_MSTR_capsule, rotation=<rot 0 -90 0>}]
 dyn_hair_bk_r_MSTR [{shapeType=Capsule, length=0.505, radius=0.505, shapeName=dyn_hair_r_MSTR_capsule, rotation=<rot 0 -90 0>}]
 dyn_hair_offset [{shapeType=Capsule, length=0.505, radius=0.505, shapeName=dyn_hair_r_MSTR_capsule, rotation=<rot 0 -90 0>}]
 dyn_hair_side_l_01 [{shapeType=Capsule, length=6.75, radius=4.25, shapeName=dyn_hair_r_01_capsule, center=<10 -2 0>, rotation=<rot 90 -63.6134 -43.6129>}]
 dyn_hair_side_l_AIM [{shapeType=Capsule, length=0.505, radius=0.505, shapeName=dyn_hair_r_MSTR_capsule, rotation=<rot 0 -90 0>}]
 dyn_hair_side_l_MSTR [{shapeType=Capsule, length=0.505, radius=0.505, shapeName=dyn_hair_r_MSTR_capsule, rotation=<rot 0 -90 0>}]
 dyn_hair_side_r_01 [{shapeType=Capsule, length=5.375, radius=5.625, shapeName=dyn_hair_r_01_capsule, center=<10 2 0>, rotation=<rot 90 63.6134 43.6141>}]
 dyn_hair_side_r_AIM [{shapeType=Capsule, length=0.505, radius=0.505, shapeName=dyn_hair_r_MSTR_capsule, rotation=<rot 0 -90 0>}]
 dyn_hair_side_r_MSTR [{shapeType=Capsule, length=0.505, radius=0.505, shapeName=dyn_hair_r_MSTR_capsule, rotation=<rot 0 -90 0>}]
 neck_01 [{
   shapeType=Capsule
   length=3.6965
   radius=7.0617
   shapeName=head_capsule
   center=<3.0345 2.5883 -0.2642>
   rotation=<rot 90 -165.9638 -155.9632>
  }]
 pelvis [{shapeType=Capsule, length=8.875, radius=14, shapeName=spine_05_capsule, center=<-2.8211 -5.3789 0>}]
 spine_02 [{shapeType=Capsule, length=6.375, radius=14.625, shapeName=spine_05_capsule, center=<0.0926 1.2984 0>}]
 spine_03 [{shapeType=Capsule, length=9.5, radius=12.25, shapeName=spine_05_capsule, center=<8.0528 1.3592 0>}]
 spine_04 [{shapeType=Capsule, length=15, radius=11.5, shapeName=spine_05_capsule, center=<6.7586 0.4016 0>}]
 spine_05 [
  {shapeType=Capsule, length=18.5, radius=9.375, shapeName=spine_05_capsule, center=<3.5089 -0.3109 0>}
  {shapeType=Capsule, length=11.875, radius=8.5, shapeName=spine_05_capsule, center=<6.6936 0.0344 0>}
 ]
 upperarm_l [{
   shapeType=Capsule
   length=20.25
   radius=6.125
   shapeName=clavicle_l_capsule
   center=<12.5437 -0.9632 0.0429>
   rotation=<rot 89.5645 92.4919 87.5101>
  }]
 upperarm_r [{
   shapeType=Capsule
   length=20.25
   radius=6.125
   shapeName=clavicle_l_capsule
   center=<-12.5437 0.9632 -0.0429>
   rotation=<rot 89.5645 92.4919 -92.4899>
  }]
}
constraints [
 {
  bone1Name=dyn_hair_bk_MSTR
  bone2Name=dyn_hair_offset
  swing1LimitDegrees=45
  swing1Motion=Free
  swing2LimitDegrees=45
  swing2Motion=Free
  twistLimitDegrees=45
  twistMotion=Free
 }
 {bone1Name=dyn_hair_bk_01, bone2Name=dyn_hair_bk_AIM, swing1LimitDegrees=30, swing2LimitDegrees=45, twistLimitDegrees=15}
 {bone1Name=dyn_hair_bk_02, bone2Name=dyn_hair_bk_01, swing1LimitDegrees=30, swing2LimitDegrees=45, twistLimitDegrees=15}
 {bone1Name=dyn_hair_bk_03, bone2Name=dyn_hair_bk_02, swing1LimitDegrees=10, swing2LimitDegrees=10, twistLimitDegrees=15}
 {bone1Name=dyn_hair_bk_04, bone2Name=dyn_hair_bk_03, swing1LimitDegrees=10, swing2LimitDegrees=2, twistLimitDegrees=15}
 {
  bone1Name=spine_05
  bone2Name=spine_04
  swing1LimitDegrees=45
  swing1Motion=Free
  swing2LimitDegrees=45
  swing2Motion=Free
  twistLimitDegrees=45
  twistMotion=Free
 }
 {
  bone1Name=neck_01
  bone2Name=spine_05
  swing1LimitDegrees=45
  swing1Motion=Free
  swing2LimitDegrees=45
  swing2Motion=Free
  twistLimitDegrees=45
  twistMotion=Free
 }
 {
  bone1Name=upperarm_r
  bone2Name=clavicle_r
  swing1LimitDegrees=45
  swing1Motion=Free
  swing2LimitDegrees=45
  swing2Motion=Free
  twistLimitDegrees=45
  twistMotion=Free
 }
 {
  bone1Name=spine_04
  bone2Name=spine_03
  swing1LimitDegrees=45
  swing1Motion=Free
  swing2LimitDegrees=45
  swing2Motion=Free
  twistLimitDegrees=45
  twistMotion=Free
 }
 {
  bone1Name=spine_03
  bone2Name=spine_02
  swing1LimitDegrees=45
  swing1Motion=Free
  swing2LimitDegrees=45
  swing2Motion=Free
  twistLimitDegrees=45
  twistMotion=Free
 }
 {
  bone1Name=clavicle_r
  bone2Name=spine_05
  swing1LimitDegrees=45
  swing1Motion=Free
  swing2LimitDegrees=45
  swing2Motion=Free
  twistLimitDegrees=45
  twistMotion=Free
 }
 {
  bone1Name=dyn_hair_bk_AIM
  bone2Name=dyn_hair_bk_MSTR
  swing1LimitDegrees=45
  swing1Motion=Free
  swing2LimitDegrees=45
  swing2Motion=Free
  twistLimitDegrees=45
  twistMotion=Free
 }
 {
  bone1Name=dyn_hair_bk_r_MSTR
  bone2Name=dyn_hair_offset
  swing1LimitDegrees=45
  swing1Motion=Free
  swing2LimitDegrees=45
  swing2Motion=Free
  twistLimitDegrees=45
  twistMotion=Free
 }
 {
  bone1Name=dyn_hair_bk_r_AIM
  bone2Name=dyn_hair_bk_r_MSTR
  swing1LimitDegrees=45
  swing1Motion=Free
  swing2LimitDegrees=45
  swing2Motion=Free
  twistLimitDegrees=45
  twistMotion=Free
 }
 {bone1Name=dyn_hair_bk_r_01, bone2Name=dyn_hair_bk_r_AIM, swing1LimitDegrees=30, swing2LimitDegrees=45, twistLimitDegrees=15}
 {bone1Name=dyn_hair_bk_r_02, bone2Name=dyn_hair_bk_r_01, swing1LimitDegrees=30, swing2LimitDegrees=45, twistLimitDegrees=15}
 {bone1Name=dyn_hair_bk_r_03, bone2Name=dyn_hair_bk_r_02, swing1LimitDegrees=10, swing2LimitDegrees=10, twistLimitDegrees=15}
 {bone1Name=dyn_hair_bk_r_04, bone2Name=dyn_hair_bk_r_03, swing1LimitDegrees=10, swing2LimitDegrees=2, twistLimitDegrees=15}
 {
  bone1Name=dyn_hair_side_r_MSTR
  bone2Name=dyn_hair_offset
  swing1LimitDegrees=45
  swing1Motion=Free
  swing2LimitDegrees=45
  swing2Motion=Free
  twistLimitDegrees=45
  twistMotion=Free
 }
 {bone1Name=dyn_hair_bk_r_05, bone2Name=dyn_hair_bk_r_04, swing1LimitDegrees=10, swing2LimitDegrees=2, twistLimitDegrees=15}
 {
  bone1Name=dyn_hair_bk_l_MSTR
  bone2Name=dyn_hair_offset
  swing1LimitDegrees=45
  swing1Motion=Free
  swing2LimitDegrees=45
  swing2Motion=Free
  twistLimitDegrees=45
  twistMotion=Free
 }
 {
  bone1Name=dyn_hair_bk_l_AIM
  bone2Name=dyn_hair_bk_l_MSTR
  swing1LimitDegrees=45
  swing1Motion=Free
  swing2LimitDegrees=45
  swing2Motion=Free
  twistLimitDegrees=45
  twistMotion=Free
 }
 {bone1Name=dyn_hair_bk_l_01, bone2Name=dyn_hair_bk_l_AIM, swing1LimitDegrees=30, swing2LimitDegrees=45, twistLimitDegrees=15}
 {bone1Name=dyn_hair_bk_l_02, bone2Name=dyn_hair_bk_l_01, swing1LimitDegrees=30, swing2LimitDegrees=45, twistLimitDegrees=15}
 {bone1Name=dyn_hair_bk_l_03, bone2Name=dyn_hair_bk_l_02, swing1LimitDegrees=10, swing2LimitDegrees=10, twistLimitDegrees=15}
 {bone1Name=dyn_hair_bk_l_04, bone2Name=dyn_hair_bk_l_03, swing1LimitDegrees=10, swing2LimitDegrees=2, twistLimitDegrees=15}
 {bone1Name=dyn_hair_bk_05, bone2Name=dyn_hair_bk_04, swing1LimitDegrees=10, swing2LimitDegrees=2, twistLimitDegrees=15}
 {bone1Name=dyn_hair_bk_06, bone2Name=dyn_hair_bk_05, swing1LimitDegrees=10, swing2LimitDegrees=2, twistLimitDegrees=15}
 {
  bone1Name=dyn_hair_side_r_AIM
  bone2Name=dyn_hair_side_r_MSTR
  swing1LimitDegrees=45
  swing1Motion=Free
  swing2LimitDegrees=45
  swing2Motion=Free
  twistLimitDegrees=45
  twistMotion=Free
 }
 {bone1Name=dyn_hair_side_r_01, bone2Name=dyn_hair_side_r_AIM, swing1LimitDegrees=45, swing2LimitDegrees=20, twistLimitDegrees=15}
 {
  bone1Name=dyn_hair_side_l_MSTR
  bone2Name=dyn_hair_offset
  swing1LimitDegrees=45
  swing1Motion=Free
  swing2LimitDegrees=45
  swing2Motion=Free
  twistLimitDegrees=45
  twistMotion=Free
 }
 {
  bone1Name=dyn_hair_side_l_AIM
  bone2Name=dyn_hair_side_l_MSTR
  swing1LimitDegrees=45
  swing1Motion=Free
  swing2LimitDegrees=45
  swing2Motion=Free
  twistLimitDegrees=45
  twistMotion=Free
 }
 {bone1Name=dyn_hair_side_l_01, bone2Name=dyn_hair_side_l_AIM, swing1LimitDegrees=45, swing2LimitDegrees=20, twistLimitDegrees=15}
 {
  bone1Name=dyn_hair_offset
  bone2Name=neck_01
  swing1LimitDegrees=45
  swing1Motion=Free
  swing2LimitDegrees=45
  swing2Motion=Free
  twistLimitDegrees=45
  twistMotion=Free
 }
 {
  bone1Name=spine_02
  bone2Name=pelvis
  swing1LimitDegrees=45
  swing1Motion=Free
  swing2LimitDegrees=45
  swing2Motion=Free
  twistLimitDegrees=45
  twistMotion=Free
 }
 {
  bone1Name=upperarm_l
  bone2Name=clavicle_l
  swing1LimitDegrees=45
  swing1Motion=Free
  swing2LimitDegrees=45
  swing2Motion=Free
  twistLimitDegrees=45
  twistMotion=Free
 }
 {
  bone1Name=clavicle_l
  bone2Name=spine_05
  swing1LimitDegrees=45
  swing1Motion=Free
  swing2LimitDegrees=45
  swing2Motion=Free
  twistLimitDegrees=45
  twistMotion=Free
 }
]
