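# Physics asset for the dynamic hair of the Snake Vest male player ensemble, named as the cape-wearing variant.
# Capsules on spine, neck, head, clavicles and upper arms form the body beneath back, side and front hair chains.
# Each chain hangs from the head through free-rotating MSTR and AIM bodies, then angle-limited joints with tight twist.
# Free-rotation constraints also tie the left and right back tips to the center back chain.
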PhysicsAsset /Game/Characters/Player/Male/Medium/Bodies/M_MED_Ensemble_Snake_Vest/Meshes/Parts/M_MED_Ensemble_Snake_Vest_FaceAcc_Dyn_Hair_Wearing_Cape
bodies {
 clavicle_l [{
   shapeType=Capsule
   length=5.25
   radius=7.5
   shapeName=clavicle_l_capsule
   center=<8.4521 -1.2518 -2.2519>
   rotation=<rot 86.2624 -0.0144 -5.0145>
  }]
 clavicle_r [{
   shapeType=Capsule
   length=4.125
   radius=7.125
   shapeName=clavicle_l_capsule
   center=<-8.0641 1.1376 0.5602>
   rotation=<rot 81.2706 2.8646 177.8503>
  }]
 dyn_hair_bk_01 [{shapeType=Capsule, length=2.5, radius=5.5, shapeName=dyn_hair_bk_01_capsule, center=<6 0 1>, rotation=<rot 0 0 90>}]
 dyn_hair_bk_02 [{shapeType=Capsule, length=4.875, radius=5, shapeName=dyn_hair_bk_02_capsule, center=<5 0 0>, rotation=<rot 90 0 0>}]
 dyn_hair_bk_03 [{shapeType=Capsule, length=4.5, radius=4.25, shapeName=dyn_hair_bk_03_capsule, center=<5 0 0>, rotation=<rot 90 0 0>}]
 dyn_hair_bk_04 [{shapeType=Capsule, length=4.5, radius=4.125, shapeName=dyn_hair_bk_04_capsule, center=<5 0 0>, rotation=<rot 90 0 0>}]
 dyn_hair_bk_05 [{shapeType=Capsule, length=4.5, radius=4.125, shapeName=dyn_hair_bk_04_capsule, center=<5 0 0>, rotation=<rot 90 0 0>}]
 dyn_hair_bk_06 [{shapeType=Capsule, length=4.5, radius=4.125, shapeName=dyn_hair_bk_04_capsule, center=<5 0 0>, rotation=<rot 90 0 0>}]
 dyn_hair_bk_AIM [{shapeType=Capsule, length=0.505, radius=0.505, shapeName=dyn_hair_bk_MSTR_capsule, rotation=<rot 0 -90 0>}]
 dyn_hair_bk_MSTR [{shapeType=Capsule, length=0.505, radius=0.505, shapeName=dyn_hair_bk_MSTR_capsule, rotation=<rot 0 -90 0>}]
 dyn_hair_bk_l_01 [{shapeType=Capsule, length=2.875, radius=4.75, shapeName=dyn_hair_r_01_capsule, center=<5 0 0>, rotation=<rot 90 0 0>}]
 dyn_hair_bk_l_02 [{shapeType=Capsule, length=3.75, radius=4.625, shapeName=dyn_hair_r_02_capsule, center=<6 0 1>, rotation=<rot 90 0 0>}]
 dyn_hair_bk_l_03 [{shapeType=Capsule, length=5.75, radius=3.5, shapeName=dyn_hair_r_03_capsule, center=<6 0 0>, rotation=<rot 90 0 0>}]
 dyn_hair_bk_l_04 [{shapeType=Capsule, length=5.375, radius=3.25, shapeName=dyn_hair_r_04_capsule, center=<6 0 0>, rotation=<rot 90 0 0>}]
 dyn_hair_bk_l_AIM [{shapeType=Capsule, length=0.505, radius=0.505, shapeName=dyn_hair_r_MSTR_capsule, rotation=<rot 0 -90 0>}]
 dyn_hair_bk_l_MSTR [{shapeType=Capsule, length=0.505, radius=0.505, shapeName=dyn_hair_r_MSTR_capsule, rotation=<rot 0 -90 0>}]
 dyn_hair_bk_r_01 [{shapeType=Capsule, length=2.875, radius=4.75, shapeName=dyn_hair_r_01_capsule, center=<5 0 0>, rotation=<rot 90 0 0>}]
 dyn_hair_bk_r_02 [{shapeType=Capsule, length=3.75, radius=4.625, shapeName=dyn_hair_r_02_capsule, center=<6 0 1>, rotation=<rot 90 0 0>}]
 dyn_hair_bk_r_03 [{shapeType=Capsule, length=5.75, radius=3.5, shapeName=dyn_hair_r_03_capsule, center=<6 0 0>, rotation=<rot 90 0 0>}]
 dyn_hair_bk_r_04 [{shapeType=Capsule, length=5.375, radius=3.25, shapeName=dyn_hair_r_04_capsule, center=<6 0 0>, rotation=<rot 90 0 0>}]
 dyn_hair_bk_r_05 [{shapeType=Capsule, length=5.375, radius=3.25, shapeName=dyn_hair_r_04_capsule, center=<6 0 0>, rotation=<rot 90 0 0>}]
 dyn_hair_bk_r_AIM [{shapeType=Capsule, length=0.505, radius=0.505, shapeName=dyn_hair_r_MSTR_capsule, rotation=<rot 0 -90 0>}]
 dyn_hair_bk_r_MSTR [{shapeType=Capsule, length=0.505, radius=0.505, shapeName=dyn_hair_r_MSTR_capsule, rotation=<rot 0 -90 0>}]
 dyn_hair_ft_l_01 [{shapeType=Capsule, length=5.25, radius=2.5, shapeName=dyn_hair_r_01_capsule, center=<5 -1 0>, rotation=<rot 90 -177.5355 172.4647>}]
 dyn_hair_ft_l_02 [{shapeType=Capsule, length=4.5, radius=2.125, shapeName=dyn_hair_r_02_capsule, center=<5 0 0>, rotation=<rot 90 73.044 63.0432>}]
 dyn_hair_ft_l_03 [{shapeType=Capsule, length=5.5, radius=1.375, shapeName=dyn_hair_r_03_capsule, center=<5 0 0>, rotation=<rot 90 166.7373 166.737>}]
 dyn_hair_ft_l_AIM [{shapeType=Capsule, length=0.505, radius=0.505, shapeName=dyn_hair_l_MSTR_capsule, rotation=<rot 0 -90 0>}]
 dyn_hair_ft_l_MSTR [{shapeType=Capsule, length=0.505, radius=0.505, shapeName=dyn_hair_l_MSTR_capsule, rotation=<rot 0 -90 0>}]
 dyn_hair_ft_r_01 [{shapeType=Capsule, length=5.25, radius=2.5, shapeName=dyn_hair_r_01_capsule, center=<5 2 0>, rotation=<rot 90 177.5355 -172.4641>}]
 dyn_hair_ft_r_02 [{shapeType=Capsule, length=4.375, radius=2.375, shapeName=dyn_hair_r_02_capsule, center=<5 0 0>, rotation=<rot 90 -77.1723 -72.172>}]
 dyn_hair_ft_r_03 [{shapeType=Capsule, length=5.125, radius=1.375, shapeName=dyn_hair_r_03_capsule, center=<5 0 0>, rotation=<rot 90 -78.2837 -83.2836>}]
 dyn_hair_ft_r_AIM [{shapeType=Capsule, length=0.505, radius=0.505, shapeName=dyn_hair_l_MSTR_capsule, rotation=<rot 0 -90 0>}]
 dyn_hair_ft_r_MSTR [{shapeType=Capsule, length=0.505, radius=0.505, shapeName=dyn_hair_l_MSTR_capsule, rotation=<rot 0 -90 0>}]
 dyn_hair_side_l_01 [{shapeType=Capsule, length=13.75, radius=4.75, shapeName=dyn_hair_r_01_capsule, center=<15 -3 0>, rotation=<rot 90 63.6134 58.6136>}]
 dyn_hair_side_l_AIM [{shapeType=Capsule, length=0.505, radius=0.505, shapeName=dyn_hair_r_MSTR_capsule, rotation=<rot 0 -90 0>}]
 dyn_hair_side_l_MSTR [{shapeType=Capsule, length=0.505, radius=0.505, shapeName=dyn_hair_r_MSTR_capsule, rotation=<rot 0 -90 0>}]
 dyn_hair_side_r_01 [{shapeType=Capsule, length=13.75, radius=4.75, shapeName=dyn_hair_r_01_capsule, center=<15 3 0>, rotation=<rot 90 63.6134 58.6136>}]
 dyn_hair_side_r_AIM [{shapeType=Capsule, length=0.505, radius=0.505, shapeName=dyn_hair_r_MSTR_capsule, rotation=<rot 0 -90 0>}]
 dyn_hair_side_r_MSTR [{shapeType=Capsule, length=0.505, radius=0.505, shapeName=dyn_hair_r_MSTR_capsule, rotation=<rot 0 -90 0>}]
 head [{shapeType=Capsule, length=6.3215, radius=6.4367, shapeName=head_capsule, center=<6.2684 -0.2133 0>, rotation=<rot 90 179.9899 179.9903>}]
 neck_01 [{shapeType=Capsule, length=3.8215, radius=6.9367, shapeName=head_capsule, center=<3.2157 3.616 0>, rotation=<rot 90 -174.0531 -164.0525>}]
 spine_02 [{shapeType=Capsule, length=3.5, radius=24.875, shapeName=spine_05_capsule, center=<-0.102 0.6577 0>}]
 spine_04 [{shapeType=Capsule, length=9.5, radius=21.25, shapeName=spine_05_capsule, center=<-2.1321 -1.9871 0>}]
 spine_05 [{shapeType=Capsule, length=19.125, radius=17.125, shapeName=spine_05_capsule, center=<-1.2692 -4.462 0>}]
 upperarm_l [{
   shapeType=Capsule
   length=23.875
   radius=7.625
   shapeName=clavicle_l_capsule
   center=<14.2843 -2.6547 -2.3561>
   rotation=<rot 90 180 179.9998>
  }]
 upperarm_r [{
   shapeType=Capsule
   length=19.25
   radius=8.375
   shapeName=clavicle_l_capsule
   center=<-13.4743 3.0682 2.3561>
   rotation=<rot 90 -144.7357 35.2641>
  }]
}
constraints [
 {
  bone1Name=dyn_hair_bk_MSTR
  bone2Name=head
  swing1LimitDegrees=45
  swing1Motion=Free
  swing2LimitDegrees=45
  swing2Motion=Free
  twistLimitDegrees=45
  twistMotion=Free
 }
 {bone1Name=dyn_hair_bk_01, bone2Name=dyn_hair_bk_AIM, swing1LimitDegrees=5, swing2LimitDegrees=25, twistLimitDegrees=2}
 {bone1Name=dyn_hair_bk_02, bone2Name=dyn_hair_bk_01, swing1LimitDegrees=5, swing2LimitDegrees=25, twistLimitDegrees=2}
 {bone1Name=dyn_hair_bk_03, bone2Name=dyn_hair_bk_02, swing1LimitDegrees=10, swing2LimitDegrees=45, twistLimitDegrees=2}
 {bone1Name=dyn_hair_bk_04, bone2Name=dyn_hair_bk_03, swing1LimitDegrees=15, swing2LimitDegrees=45, twistLimitDegrees=2}
 {
  bone1Name=head
  bone2Name=neck_01
  swing1LimitDegrees=45
  swing1Motion=Free
  swing2LimitDegrees=45
  swing2Motion=Free
  twistLimitDegrees=45
  twistMotion=Free
 }
 {
  bone1Name=spine_05
  bone2Name=spine_04
  swing1LimitDegrees=45
  swing1Motion=Free
  swing2LimitDegrees=45
  swing2Motion=Free
  twistLimitDegrees=45
  twistMotion=Free
 }
 {
  bone1Name=neck_01
  bone2Name=spine_05
  swing1LimitDegrees=45
  swing1Motion=Free
  swing2LimitDegrees=45
  swing2Motion=Free
  twistLimitDegrees=45
  twistMotion=Free
 }
 {
  bone1Name=upperarm_r
  bone2Name=clavicle_r
  swing1LimitDegrees=45
  swing1Motion=Free
  swing2LimitDegrees=45
  swing2Motion=Free
  twistLimitDegrees=45
  twistMotion=Free
 }
 {
  bone1Name=dyn_hair_ft_r_MSTR
  bone2Name=head
  swing1LimitDegrees=45
  swing1Motion=Free
  swing2LimitDegrees=45
  swing2Motion=Free
  twistLimitDegrees=45
  twistMotion=Free
 }
 {bone1Name=dyn_hair_ft_r_01, bone2Name=dyn_hair_ft_r_AIM, swing1LimitDegrees=25, swing2LimitDegrees=15, twistLimitDegrees=2}
 {bone1Name=dyn_hair_ft_r_02, bone2Name=dyn_hair_ft_r_01, swing1LimitDegrees=25, swing2LimitDegrees=15, twistLimitDegrees=2}
 {bone1Name=dyn_hair_ft_r_03, bone2Name=dyn_hair_ft_r_02, swing1LimitDegrees=25, swing2LimitDegrees=15, twistLimitDegrees=2}
 {
  bone1Name=dyn_hair_ft_l_MSTR
  bone2Name=head
  swing1LimitDegrees=45
  swing1Motion=Free
  swing2LimitDegrees=45
  swing2Motion=Free
  twistLimitDegrees=45
  twistMotion=Free
 }
 {bone1Name=dyn_hair_ft_l_01, bone2Name=dyn_hair_ft_l_AIM, swing1LimitDegrees=35, swing2LimitDegrees=5, twistLimitDegrees=2}
 {bone1Name=dyn_hair_ft_l_02, bone2Name=dyn_hair_ft_l_01, swing1LimitDegrees=45, swing2LimitDegrees=5, twistLimitDegrees=2}
 {bone1Name=dyn_hair_ft_l_03, bone2Name=dyn_hair_ft_l_02, swing1LimitDegrees=45, swing2LimitDegrees=5, twistLimitDegrees=2}
 {
  bone1Name=clavicle_l
  bone2Name=spine_05
  swing1LimitDegrees=45
  swing1Motion=Free
  swing2LimitDegrees=45
  swing2Motion=Free
  twistLimitDegrees=45
  twistMotion=Free
 }
 {
  bone1Name=clavicle_r
  bone2Name=spine_05
  swing1LimitDegrees=45
  swing1Motion=Free
  swing2LimitDegrees=45
  swing2Motion=Free
  twistLimitDegrees=45
  twistMotion=Free
 }
 {
  bone1Name=upperarm_l
  bone2Name=clavicle_l
  swing1LimitDegrees=45
  swing1Motion=Free
  swing2LimitDegrees=45
  swing2Motion=Free
  twistLimitDegrees=45
  twistMotion=Free
 }
 {
  bone1Name=dyn_hair_bk_AIM
  bone2Name=dyn_hair_bk_MSTR
  swing1LimitDegrees=45
  swing1Motion=Free
  swing2LimitDegrees=45
  swing2Motion=Free
  twistLimitDegrees=45
  twistMotion=Free
 }
 {
  bone1Name=dyn_hair_bk_r_MSTR
  bone2Name=head
  swing1LimitDegrees=45
  swing1Motion=Free
  swing2LimitDegrees=45
  swing2Motion=Free
  twistLimitDegrees=45
  twistMotion=Free
 }
 {
  bone1Name=dyn_hair_bk_r_AIM
  bone2Name=dyn_hair_bk_r_MSTR
  swing1LimitDegrees=45
  swing1Motion=Free
  swing2LimitDegrees=45
  swing2Motion=Free
  twistLimitDegrees=45
  twistMotion=Free
 }
 {bone1Name=dyn_hair_bk_r_01, bone2Name=dyn_hair_bk_r_AIM, swing1LimitDegrees=5, swing2LimitDegrees=25, twistLimitDegrees=2}
 {bone1Name=dyn_hair_bk_r_02, bone2Name=dyn_hair_bk_r_01, swing1LimitDegrees=5, swing2LimitDegrees=25, twistLimitDegrees=2}
 {bone1Name=dyn_hair_bk_r_03, bone2Name=dyn_hair_bk_r_02, swing1LimitDegrees=10, swing2LimitDegrees=45, twistLimitDegrees=2}
 {bone1Name=dyn_hair_bk_r_04, bone2Name=dyn_hair_bk_r_03, swing1LimitDegrees=15, swing2LimitDegrees=45, twistLimitDegrees=2}
 {
  bone1Name=dyn_hair_side_r_MSTR
  bone2Name=head
  swing1LimitDegrees=45
  swing1Motion=Free
  swing2LimitDegrees=45
  swing2Motion=Free
  twistLimitDegrees=45
  twistMotion=Free
 }
 {
  bone1Name=dyn_hair_ft_r_AIM
  bone2Name=dyn_hair_ft_r_MSTR
  swing1LimitDegrees=45
  swing1Motion=Free
  swing2LimitDegrees=45
  swing2Motion=Free
  twistLimitDegrees=45
  twistMotion=Free
 }
 {bone1Name=dyn_hair_bk_r_05, bone2Name=dyn_hair_bk_r_04, swing1LimitDegrees=15, swing2LimitDegrees=45, twistLimitDegrees=2}
 {
  bone1Name=dyn_hair_bk_l_MSTR
  bone2Name=head
  swing1LimitDegrees=45
  swing1Motion=Free
  swing2LimitDegrees=45
  swing2Motion=Free
  twistLimitDegrees=45
  twistMotion=Free
 }
 {
  bone1Name=dyn_hair_bk_l_AIM
  bone2Name=dyn_hair_bk_l_MSTR
  swing1LimitDegrees=45
  swing1Motion=Free
  swing2LimitDegrees=45
  swing2Motion=Free
  twistLimitDegrees=45
  twistMotion=Free
 }
 {bone1Name=dyn_hair_bk_l_01, bone2Name=dyn_hair_bk_l_AIM, swing1LimitDegrees=5, swing2LimitDegrees=25, twistLimitDegrees=2}
 {bone1Name=dyn_hair_bk_l_02, bone2Name=dyn_hair_bk_l_01, swing1LimitDegrees=5, swing2LimitDegrees=25, twistLimitDegrees=2}
 {bone1Name=dyn_hair_bk_l_03, bone2Name=dyn_hair_bk_l_02, swing1LimitDegrees=10, swing2LimitDegrees=45, twistLimitDegrees=2}
 {bone1Name=dyn_hair_bk_l_04, bone2Name=dyn_hair_bk_l_03, swing1LimitDegrees=15, swing2LimitDegrees=45, twistLimitDegrees=2}
 {bone1Name=dyn_hair_bk_05, bone2Name=dyn_hair_bk_04, swing1LimitDegrees=15, swing2LimitDegrees=45, twistLimitDegrees=2}
 {bone1Name=dyn_hair_bk_06, bone2Name=dyn_hair_bk_05, swing1LimitDegrees=15, swing2LimitDegrees=45, twistLimitDegrees=2}
 {
  bone1Name=dyn_hair_bk_l_04
  bone2Name=dyn_hair_bk_04
  swing1LimitDegrees=45
  swing1Motion=Free
  swing2LimitDegrees=45
  swing2Motion=Free
  twistLimitDegrees=45
  twistMotion=Free
 }
 {
  bone1Name=dyn_hair_bk_r_05
  bone2Name=dyn_hair_bk_05
  swing1LimitDegrees=45
  swing1Motion=Free
  swing2LimitDegrees=45
  swing2Motion=Free
  twistLimitDegrees=45
  twistMotion=Free
 }
 {
  bone1Name=dyn_hair_side_r_AIM
  bone2Name=dyn_hair_side_r_MSTR
  swing1LimitDegrees=45
  swing1Motion=Free
  swing2LimitDegrees=45
  swing2Motion=Free
  twistLimitDegrees=45
  twistMotion=Free
 }
 {bone1Name=dyn_hair_side_r_01, bone2Name=dyn_hair_side_r_AIM, swing1LimitDegrees=30, swing2LimitDegrees=15, twistLimitDegrees=2}
 {
  bone1Name=dyn_hair_side_l_MSTR
  bone2Name=head
  swing1LimitDegrees=45
  swing1Motion=Free
  swing2LimitDegrees=45
  swing2Motion=Free
  twistLimitDegrees=45
  twistMotion=Free
 }
 {
  bone1Name=dyn_hair_side_l_AIM
  bone2Name=dyn_hair_side_l_MSTR
  swing1LimitDegrees=45
  swing1Motion=Free
  swing2LimitDegrees=45
  swing2Motion=Free
  twistLimitDegrees=45
  twistMotion=Free
 }
 {bone1Name=dyn_hair_side_l_01, bone2Name=dyn_hair_side_l_AIM, swing1LimitDegrees=30, swing2LimitDegrees=15, twistLimitDegrees=2}
 {
  bone1Name=dyn_hair_ft_l_AIM
  bone2Name=dyn_hair_ft_l_MSTR
  swing1LimitDegrees=45
  swing1Motion=Free
  swing2LimitDegrees=45
  swing2Motion=Free
  twistLimitDegrees=45
  twistMotion=Free
 }
 {
  bone1Name=spine_04
  bone2Name=spine_02
  swing1LimitDegrees=45
  swing1Motion=Free
  swing2LimitDegrees=45
  swing2Motion=Free
  twistLimitDegrees=45
  twistMotion=Free
 }
]
