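# Post-process style hair AnimBP for the Snake Vest male outfit's FaceAcc part, copying body poses and adding secondary motion.
# Hair tips use AnimDynamics; front hair uses RigidBody, feeding cached back and side hair variants for cape, backbling or neither.
# EventGraph derives head-rotation alphas that drive corrective bone offsets before and after simulation, plus a gravity override.
# Bool blends select variants by worn items and skydiving, with a constraint-only fallback when rigid bodies are disabled.
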AnimBlueprint /BRCosmetics/Characters/Player/Male/Medium/Bodies/M_MED_Ensemble_Snake_Vest/Meshes/Parts/M_MED_Ensemble_Snake_Vest_FaceAcc_AnimBP
target_skeleton=&/M_MED_Ensemble_Snake_Vest_FaceAcc_Skeleton
graphs [
 AnimGraph {nodes [
   n1 = Look At - Bone: dyn_hair_bk_l_04(Alpha=1) {
    alphaBoolBlend {blendOption=Linear}
    boneToModify {boneName=dyn_hair_bk_l_04}
    lookAtLocation=<-20 -15.1331 -7.6525>
    lookAtTarget {boneReference {boneName=spine_01}}
    lookAt_Axis {axis=<1 0>}
   }
   n2 = Look At - Bone: dyn_hair_bk_r_04(ComponentPose=@n1, Alpha=1) {
    alphaBoolBlend {blendOption=Linear}
    boneToModify {boneName=dyn_hair_bk_r_04}
    lookAtLocation=<-20 -15.1396 7.1042>
    lookAtTarget {boneReference {boneName=spine_01}}
    lookAt_Axis {axis=<1 0>}
   }
   Look At - Bone: dyn_hair_bk_04(ComponentPose=@n2, Alpha=1) {
    alphaBoolBlend {blendOption=Linear}
    boneToModify {boneName=dyn_hair_bk_04}
    lookAtLocation=<-20 -17.439 0>
    lookAtTarget {boneReference {boneName=spine_01}}
    lookAt_Axis {axis=<1 0>}
   }
   n3 = Look At - Bone: dyn_hair_side_r_AIM(Alpha=0) {
    alphaBoolBlend {blendOption=Linear}
    boneToModify {boneName=dyn_hair_side_r_AIM}
    lookAtLocation=<-30 -70 16>
    lookAtTarget {boneReference {boneName=spine_05}}
    lookAt_Axis {axis=<1 0>}
   }
   Look At - Bone: dyn_hair_side_l_AIM(ComponentPose=@n3, Alpha=0) {
    alphaBoolBlend {blendOption=Linear}
    boneToModify {boneName=dyn_hair_side_l_AIM}
    lookAtLocation=<-30 -72 -16>
    lookAtTarget {boneReference {boneName=spine_05}}
    lookAt_Axis {axis=<1 0>}
   }
   n4 = Look At - Bone: dyn_hair_bk_l_AIM(Alpha=0) {
    alphaBoolBlend {blendOption=Linear}
    boneToModify {boneName=dyn_hair_bk_l_AIM}
    lookAtLocation=<-20 -40.6839 -8.2203>
    lookAtTarget {boneReference {boneName=spine_02}}
    lookAt_Axis {axis=<1 0>}
   }
   n5 = Look At - Bone: dyn_hair_bk_r_AIM(ComponentPose=@n4, Alpha=0) {
    alphaBoolBlend {blendOption=Linear}
    boneToModify {boneName=dyn_hair_bk_r_AIM}
    lookAtLocation=<-20 -40.3672 8.3>
    lookAtTarget {boneReference {boneName=spine_02}}
    lookAt_Axis {axis=<1 0>}
   }
   Look At - Bone: dyn_hair_bk_AIM(ComponentPose=@n5, Alpha=0) {
    alphaBoolBlend {blendOption=Linear}
    boneToModify {boneName=dyn_hair_bk_AIM}
    lookAtLocation=<-20 -35>
    lookAtTarget {boneReference {boneName=spine_02}}
    lookAt_Axis {axis=<1 0>}
   }
   n6 = Get Head_Rot_Bk() -> (Head_Rot_Bk:double)
   n7 = "Transform (Modify) Bone - Bone: dyn_hair_side_r_MSTR"(Alpha=@n6, Translation="0,2,0", Rotation="0,30,0", Scale="1,1,1") {
    rotationMode=BMM_Additive
    rotationSpace=BCS_BoneSpace
    translationMode=BMM_Additive
    translationSpace=BCS_BoneSpace
    alphaBoolBlend {blendOption=Linear}
    boneToModify {boneName=dyn_hair_side_r_MSTR}
    rotation {yaw=30}
    translation {y=2}
   }
   "Transform (Modify) Bone - Bone: dyn_hair_side_l_MSTR"(ComponentPose=@n7, Alpha=@n6, Translation="0,-2,0", Rotation="0,-30,0", Scale="1,1,1") {
    rotationMode=BMM_Additive
    rotationSpace=BCS_BoneSpace
    scaleMode=BMM_Additive
    scaleSpace=BCS_BoneSpace
    translationMode=BMM_Additive
    translationSpace=BCS_BoneSpace
    alphaBoolBlend {blendOption=Linear}
    boneToModify {boneName=dyn_hair_side_l_MSTR}
    rotation {yaw=-30}
    translation {y=-2}
   }
   n8 = Blend Poses by bool(
    BlendPose_0=Use cached pose 'wearingCapePose'()
    BlendPose_1=Blend Poses by bool(
     BlendPose_0=Use cached pose 'wearingBackblingPose'()
     BlendPose_1=Use cached pose 'noWearingPose'()
     BlendTime_0=0.3
     BlendTime_1=0.3
     bActiveValue=Get bIsSkyDiving()
    ) {blendType=Linear}
    BlendTime_0=0.3
    BlendTime_1=0.3
    bActiveValue=Get isWearCape() -> (isWearCape:bool)
   ) {blendType=Linear}
   n9 = Use cached pose 'noWearingPose'()
   n10 = Get isWearBackbling() -> (isWearBackbling:bool)
   Output Pose(Result=Blend Poses by bool(
     BlendPose_0=Blend Poses by bool(BlendPose_0=@n8, BlendPose_1=@n9, BlendTime_0=0.3, BlendTime_1=0.3, bActiveValue=@n10) {blendType=Linear}
     BlendPose_1=Use cached pose 'noRigidBodyPose'()
     BlendTime_0=0.3
     BlendTime_1=0.3
     bActiveValue=Get bRigidBodyEnabled()
    ) {blendType=Linear}) {layerGroup=None}
   n11 = Layered blend per bone(BasePose=Copy Pose From Mesh() {bUseAttachedParent=true}, BlendPoses_0=Copy Pose From Mesh(SourceMeshComponent=Get Body_Mesh())) {layerSetup [{branchFilters [{boneName=upperarm_twist_01_l}, {boneName=upperarm_twist_01_r}, {boneName=upperarm_l}, {boneName=upperarm_r}]}]}
   n12 = Anim Dynamics - Bone: dyn_hair_tips_01!(ComponentPose=Local To Component(LocalPose=@n11), Alpha=1, GravityScale:float=1, GravityOverride:Vector="0,0,-980", ExternalForce="0,0,0") -> (Pose) {
    angularDampingOverride=0.9999
    angularSpringConstant=500
    bAngularSpring=true
    bGravityOverrideInSimSpace=true
    bOverrideAngularDamping=true
    simulationSpace=RootRelative
    alphaBoolBlend {blendOption=Linear}
    boundBone {boneName=dyn_hair_tips_01}
    componentAppliedLinearAccClamp=<5000 5000 5000>
    componentLinearAccScale=<0.6 0.6 0.6>
    componentLinearVelScale=<0.6 0.6 0.6>
    gravityOverride {z=-980}
    physicsBodyDefinitions [{
      sphereCollisionRadius=0
      boundBone {boneName=dyn_hair_tips_01}
      boxExtents=<12 16 16>
      constraintSetup {angularLimitsMax=<0 5 20>, angularLimitsMin=<0 -20 -20>, angularTarget=<0 0 0>}
      localJointOffset=<1.5 0 0>
     }]
    relativeSpaceBone {boneName=head}
   }
   n13 = Anim Dynamics - Bone: dyn_hair_tips_02!(ComponentPose=@n12, Alpha=1, GravityScale:float=1, GravityOverride:Vector="0,0,-980", ExternalForce="0,0,0") -> (Pose) {
    angularDampingOverride=1
    angularSpringConstant=500
    bAngularSpring=true
    bGravityOverrideInSimSpace=true
    bOverrideAngularDamping=true
    simulationSpace=RootRelative
    alphaBoolBlend {blendOption=Linear}
    boundBone {boneName=dyn_hair_tips_02}
    componentAppliedLinearAccClamp=<3000 3000 3000>
    componentLinearAccScale=<0.6 0.6 0.6>
    componentLinearVelScale=<0.6 0.6 0.6>
    gravityOverride {z=-980}
    physicsBodyDefinitions [{
      sphereCollisionRadius=0
      boundBone {boneName=dyn_hair_tips_02}
      boxExtents=<14 16 16>
      constraintSetup {angularLimitsMax=<0 5 15>, angularLimitsMin=<0 -25 -15>, angularTarget=<0 0 0>}
      localJointOffset=<3 0 0>
     }]
    relativeSpaceBone {boneName=head}
   }
   n14 = Anim Dynamics - Bone: dyn_hair_tips_03!(ComponentPose=@n13, Alpha=1, GravityScale:float=1, GravityOverride:Vector="0,0,-980", ExternalForce="0,0,0") -> (Pose) {
    angularDampingOverride=1
    angularSpringConstant=500
    bAngularSpring=true
    bGravityOverrideInSimSpace=true
    bOverrideAngularDamping=true
    simulationSpace=RootRelative
    alphaBoolBlend {blendOption=Linear}
    boundBone {boneName=dyn_hair_tips_03}
    componentAppliedLinearAccClamp=<3000 3000 3000>
    componentLinearAccScale=<0.6 0.6 0.6>
    componentLinearVelScale=<0.6 0.6 0.6>
    gravityOverride {z=-980}
    physicsBodyDefinitions [{
      sphereCollisionRadius=0
      boundBone {boneName=dyn_hair_tips_03}
      boxExtents=<14 16 16>
      constraintSetup {angularLimitsMax=<0 5 15>, angularLimitsMin=<0 -25 -15>, angularTarget=<0 0 0>}
      localJointOffset=<3 0 0>
     }]
    relativeSpaceBone {boneName=head}
   }
   Save cached pose 'BasePose'(Pose=Component To Local(ComponentPose=@n14)) {cacheName=BasePose}
   n15 = Look At - Bone: dyn_hair_ft_r_AIM(ComponentPose=Local To Component(LocalPose=Use cached pose 'BasePose'()), Alpha=0.5) {
    alphaBoolBlend {blendOption=Linear}
    boneToModify {boneName=dyn_hair_ft_r_AIM}
    lookAtLocation=<-20 12.6 8.5>
    lookAtTarget {boneReference {boneName=spine_05}}
    lookAt_Axis {axis=<1 0>}
   }
   n16 = Look At - Bone: dyn_hair_ft_l_AIM(ComponentPose=@n15, Alpha=0.5) {
    alphaBoolBlend {blendOption=Linear}
    boneToModify {boneName=dyn_hair_ft_l_AIM}
    lookAtLocation=<-20 11 -8.5>
    lookAtTarget {boneReference {boneName=spine_05}}
    lookAt_Axis {axis=<1 0>}
   }
   n17 = Get Head_Rot_Bk()
   n18 = "Transform (Modify) Bone - Bone: dyn_hair_ft_l_MSTR"(ComponentPose=@n16, Alpha=@n17, Translation="0,0,0", Rotation="0,-30,0", Scale="1,1,1") {
    rotationMode=BMM_Additive
    rotationSpace=BCS_BoneSpace
    scaleMode=BMM_Additive
    scaleSpace=BCS_BoneSpace
    translationMode=BMM_Additive
    translationSpace=BCS_BoneSpace
    alphaBoolBlend {blendOption=Linear}
    boneToModify {boneName=dyn_hair_ft_l_MSTR}
    rotation {yaw=-30}
   }
   n19 = "Transform (Modify) Bone - Bone: dyn_hair_ft_r_MSTR"(ComponentPose=@n18, Alpha=@n17, Translation="0,0,0", Rotation="0,30.000,0", Scale="1,1,1") {
    rotationMode=BMM_Additive
    rotationSpace=BCS_BoneSpace
    scaleMode=BMM_Additive
    scaleSpace=BCS_BoneSpace
    translationMode=BMM_Additive
    translationSpace=BCS_BoneSpace
    alphaBoolBlend {blendOption=Linear}
    boneToModify {boneName=dyn_hair_ft_r_MSTR}
    rotation {yaw=30}
   }
   n20 = Get Head_Rot_Dn() -> (Head_Rot_Dn:double)
   n21 = "Transform (Modify) Bone - Bone: dyn_hair_ft_r_MSTR"(ComponentPose=@n19, Alpha=@n20, Translation="0,0,0", Rotation="0.000,-50.000,0", Scale="1,1,1") {
    rotationMode=BMM_Additive
    rotationSpace=BCS_BoneSpace
    scaleMode=BMM_Additive
    scaleSpace=BCS_BoneSpace
    translationMode=BMM_Additive
    translationSpace=BCS_BoneSpace
    alphaBoolBlend {blendOption=Linear}
    boneToModify {boneName=dyn_hair_ft_r_MSTR}
    rotation {yaw=-50}
   }
   n22 = "Transform (Modify) Bone - Bone: dyn_hair_ft_l_MSTR"(ComponentPose=@n21, Alpha=@n20, Translation="0,0,0", Rotation="20,40,0", Scale="1,1,1") {
    rotationMode=BMM_Additive
    rotationSpace=BCS_BoneSpace
    scaleMode=BMM_Additive
    scaleSpace=BCS_BoneSpace
    translationMode=BMM_Additive
    translationSpace=BCS_BoneSpace
    alphaBoolBlend {blendOption=Linear}
    boneToModify {boneName=dyn_hair_ft_l_MSTR}
    rotation {pitch=20, yaw=40}
   }
   n23 = Get Head_Rot_R() -> (Head_Rot_R:double)
   n24 = "Transform (Modify) Bone - Bone: dyn_hair_ft_l_MSTR"(ComponentPose=@n22, Alpha=@n23, Translation="3,0,1", Rotation="-10,0,0", Scale="1,1,1") {
    rotationMode=BMM_Additive
    rotationSpace=BCS_BoneSpace
    scaleMode=BMM_Additive
    scaleSpace=BCS_BoneSpace
    translationMode=BMM_Additive
    translationSpace=BCS_BoneSpace
    alphaBoolBlend {blendOption=Linear}
    boneToModify {boneName=dyn_hair_ft_l_MSTR}
    rotation {pitch=-10}
    translation {x=3, z=1}
   }
   n25 = "Transform (Modify) Bone - Bone: dyn_hair_ft_r_MSTR"(ComponentPose=@n24, Alpha=@n23, Translation="0,2,2", Rotation="0,20.000,0", Scale="1,1,1") {
    rotationMode=BMM_Additive
    rotationSpace=BCS_BoneSpace
    scaleMode=BMM_Additive
    scaleSpace=BCS_BoneSpace
    translationMode=BMM_Additive
    translationSpace=BCS_BoneSpace
    alphaBoolBlend {blendOption=Linear}
    boneToModify {boneName=dyn_hair_ft_r_MSTR}
    rotation {yaw=20}
    translation {y=2, z=2}
   }
   n26 = Component To Local(ComponentPose=@n25)
   n27 = Look At - Bone: dyn_hair_ft_r_AIM(ComponentPose=Local To Component(LocalPose=Use cached pose 'BasePose'()), Alpha=0.5) {
    alphaBoolBlend {blendOption=Linear}
    boneToModify {boneName=dyn_hair_ft_r_AIM}
    lookAtLocation=<-20 12.6 8.5>
    lookAtTarget {boneReference {boneName=spine_05}}
    lookAt_Axis {axis=<1 0>}
   }
   n28 = Look At - Bone: dyn_hair_ft_l_AIM(ComponentPose=@n27, Alpha=0.5) {
    alphaBoolBlend {blendOption=Linear}
    boneToModify {boneName=dyn_hair_ft_l_AIM}
    lookAtLocation=<-20 11 -8.5>
    lookAtTarget {boneReference {boneName=spine_05}}
    lookAt_Axis {axis=<1 0>}
   }
   n29 = Get Head_Rot_Dn()
   n30 = "Transform (Modify) Bone - Bone: dyn_hair_ft_r_MSTR"(ComponentPose=@n28, Alpha=@n29, Translation="0,0,0", Rotation="0.000,-20,0", Scale="1,1,1") {
    rotationMode=BMM_Additive
    rotationSpace=BCS_BoneSpace
    scaleMode=BMM_Additive
    scaleSpace=BCS_BoneSpace
    translationMode=BMM_Additive
    translationSpace=BCS_BoneSpace
    alphaBoolBlend {blendOption=Linear}
    boneToModify {boneName=dyn_hair_ft_r_MSTR}
    rotation {yaw=-20}
   }
   n31 = "Transform (Modify) Bone - Bone: dyn_hair_ft_l_MSTR"(ComponentPose=@n30, Alpha=@n29, Translation="0,0,0", Rotation="0,10,0", Scale="1,1,1") {
    rotationMode=BMM_Additive
    rotationSpace=BCS_BoneSpace
    scaleMode=BMM_Additive
    scaleSpace=BCS_BoneSpace
    translationMode=BMM_Additive
    translationSpace=BCS_BoneSpace
    alphaBoolBlend {blendOption=Linear}
    boneToModify {boneName=dyn_hair_ft_l_MSTR}
    rotation {yaw=10}
   }
   n32 = Component To Local(ComponentPose=@n31)
   n33 = Get bIsInFrontEnd()
   n34 = RigidBody!(
    ComponentPose=Local To Component(LocalPose=Blend Poses by bool(BlendPose_0=@n32, BlendPose_1=@n26, BlendTime_0=0.3, BlendTime_1=0.3, bActiveValue=@n33) {blendType=Linear})
    Alpha
    OverrideWorldGravity:Vector="0,0,-700"
    ExternalForce
    ComponentLinearAccScale:Vector=Get AccScaleOverride() -> (AccScaleOverride:Vector)
   ) -> (Pose) {
    bClampLinearTranslationLimitToRefPose=true
    bOverrideWorldGravity=true
    overridePhysicsAsset=&/M_MED_Ensemble_Snake_Vest_FaceAcc_Dyn_Front_Hair
    simulationSpace=BaseBoneSpace
    alphaBoolBlend {blendOption=Linear}
    alphaScaleBiasClamp {interpSpeedDecreasing=4, interpSpeedIncreasing=6}
    baseBoneRef {boneName=spine_05}
    componentAppliedLinearAccClamp=<1000 1000 800>
    componentLinearAccScale=<0.5 0.18 0.1>
    componentLinearVelScale=<0.5 0.5 0.5>
    overrideWorldGravity {z=-875}
    simSpaceSettings {maxAngularAcceleration=200, maxAngularVelocity=5, maxLinearAcceleration=500, maxLinearVelocity=200}
   }
   Save cached pose 'frontHairRBPose'(Pose=Component To Local(ComponentPose=@n34)) {cacheName=frontHairRBPose}
   n35 = Get Head_Rot_Bk()
   n36 = "Transform (Modify) Bone - Bone: dyn_hair_bk_MSTR"(
    ComponentPose=Local To Component(LocalPose=Use cached pose 'frontHairRBPose'())
    Alpha=@n35
    Translation="0,0,0"
    Rotation="15,0,0"
    Scale="1,1,1"
   ) {
    rotationMode=BMM_Additive
    rotationSpace=BCS_BoneSpace
    translationMode=BMM_Additive
    translationSpace=BCS_BoneSpace
    alphaBoolBlend {blendOption=Linear}
    boneToModify {boneName=dyn_hair_bk_MSTR}
    rotation {pitch=15}
   }
   n37 = "Transform (Modify) Bone - Bone: dyn_hair_bk_r_MSTR"(ComponentPose=@n36, Alpha=@n35, Translation="0,0,5", Rotation="10,0,0", Scale="1,1,1") {
    rotationMode=BMM_Additive
    rotationSpace=BCS_BoneSpace
    translationMode=BMM_Additive
    translationSpace=BCS_BoneSpace
    alphaBoolBlend {blendOption=Linear}
    boneToModify {boneName=dyn_hair_bk_r_MSTR}
    rotation {pitch=10}
    translation {z=5}
   }
   n38 = "Transform (Modify) Bone - Bone: dyn_hair_bk_l_MSTR"(ComponentPose=@n37, Alpha=@n35, Translation="0,0,5", Rotation="10,0,0", Scale="1,1,1") {
    rotationMode=BMM_Additive
    rotationSpace=BCS_BoneSpace
    translationMode=BMM_Additive
    translationSpace=BCS_BoneSpace
    alphaBoolBlend {blendOption=Linear}
    boneToModify {boneName=dyn_hair_bk_l_MSTR}
    rotation {pitch=10}
    translation {z=5}
   }
   n39 = Get Head_Rot_Bk()
   n40 = "Transform (Modify) Bone - Bone: dyn_hair_side_r_MSTR"(ComponentPose=@n38, Alpha=@n39, Translation="0,0,0", Rotation="0,50,0", Scale="1,1,1") {
    rotationMode=BMM_Additive
    rotationSpace=BCS_BoneSpace
    translationMode=BMM_Additive
    translationSpace=BCS_BoneSpace
    alphaBoolBlend {blendOption=Linear}
    boneToModify {boneName=dyn_hair_side_r_MSTR}
    rotation {yaw=50}
   }
   n41 = "Transform (Modify) Bone - Bone: dyn_hair_side_l_MSTR"(ComponentPose=@n40, Alpha=@n39, Translation="0,0,0", Rotation="0,-50,0", Scale="1,1,1") {
    rotationMode=BMM_Additive
    rotationSpace=BCS_BoneSpace
    scaleMode=BMM_Additive
    scaleSpace=BCS_BoneSpace
    translationMode=BMM_Additive
    translationSpace=BCS_BoneSpace
    alphaBoolBlend {blendOption=Linear}
    boneToModify {boneName=dyn_hair_side_l_MSTR}
    rotation {yaw=-50}
   }
   n42 = "Transform (Modify) Bone - Bone: dyn_hair_bk_MSTR"(ComponentPose=@n41, Alpha=Get Head_Rot_Dn(), Translation="0,0,0", Rotation="-5,0,0", Scale="1,1,1") {
    rotationMode=BMM_Additive
    rotationSpace=BCS_BoneSpace
    translationMode=BMM_Additive
    translationSpace=BCS_BoneSpace
    alphaBoolBlend {blendOption=Linear}
    boneToModify {boneName=dyn_hair_bk_MSTR}
    rotation {pitch=-5}
   }
   n43 = Get Head_Rot_R()
   n44 = "Transform (Modify) Bone - Bone: dyn_hair_bk_MSTR"(ComponentPose=@n42, Alpha=@n43, Translation="-3,0,2", Rotation="30,-15,0", Scale="1,1,1") {
    rotationMode=BMM_Additive
    rotationSpace=BCS_BoneSpace
    translationMode=BMM_Additive
    translationSpace=BCS_BoneSpace
    alphaBoolBlend {blendOption=Linear}
    boneToModify {boneName=dyn_hair_bk_MSTR}
    rotation {pitch=30, yaw=-15}
    translation {x=-3, z=2}
   }
   n45 = "Transform (Modify) Bone - Bone: dyn_hair_bk_r_MSTR"(ComponentPose=@n44, Alpha=@n43, Translation="0,2,5", Rotation="60,-20,0", Scale="1,1,1") {
    rotationMode=BMM_Additive
    rotationSpace=BCS_BoneSpace
    translationMode=BMM_Additive
    translationSpace=BCS_BoneSpace
    alphaBoolBlend {blendOption=Linear}
    boneToModify {boneName=dyn_hair_bk_r_MSTR}
    rotation {pitch=60, yaw=-20}
    translation {y=2, z=5}
   }
   n46 = Get Head_Rot_R()
   n47 = "Transform (Modify) Bone - Bone: dyn_hair_side_r_MSTR"(ComponentPose=@n45, Alpha=@n46, Translation="0,0,0", Rotation="6.385,37.539,0", Scale="1,1,1") {
    rotationMode=BMM_Additive
    rotationSpace=BCS_BoneSpace
    scaleMode=BMM_Additive
    scaleSpace=BCS_BoneSpace
    translationMode=BMM_Additive
    translationSpace=BCS_BoneSpace
    alphaBoolBlend {blendOption=Linear}
    boneToModify {boneName=dyn_hair_side_r_MSTR}
    rotation {pitch=6.385, yaw=37.539}
   }
   n48 = "Transform (Modify) Bone - Bone: dyn_hair_side_r_AIM"(ComponentPose=@n47, Alpha=@n46, Translation="0,2,0", Rotation="0,0,25", Scale="1,1,1") {
    rotationMode=BMM_Additive
    rotationSpace=BCS_BoneSpace
    scaleMode=BMM_Additive
    scaleSpace=BCS_BoneSpace
    translationMode=BMM_Additive
    translationSpace=BCS_BoneSpace
    alphaBoolBlend {blendOption=Linear}
    boneToModify {boneName=dyn_hair_side_r_AIM}
    rotation {roll=25}
    translation {y=2}
   }
   n49 = "Transform (Modify) Bone - Bone: dyn_hair_ft_l_AIM"(ComponentPose=@n48, Alpha=@n46, Translation="0,0,0", Rotation="18.047,26.389,8.738", Scale="1,1,1") {
    rotationMode=BMM_Additive
    rotationSpace=BCS_BoneSpace
    scaleMode=BMM_Additive
    scaleSpace=BCS_BoneSpace
    translationMode=BMM_Additive
    translationSpace=BCS_BoneSpace
    alphaBoolBlend {blendOption=Linear}
    boneToModify {boneName=dyn_hair_ft_l_AIM}
    rotation=<rot 18.047 26.3891 8.738>
   }
   n50 = Look At - Bone: dyn_hair_bk_l_AIM(ComponentPose=@n49, Alpha=1) {
    alphaBoolBlend {blendOption=Linear}
    boneToModify {boneName=dyn_hair_bk_l_AIM}
    lookAtLocation=<-19.9968 -34.4087 -7.9296>
    lookAtTarget {boneReference {boneName=spine_04}}
    lookAt_Axis {axis=<1 0>}
   }
   n51 = Look At - Bone: dyn_hair_bk_r_AIM(ComponentPose=@n50, Alpha=1) {
    alphaBoolBlend {blendOption=Linear}
    boneToModify {boneName=dyn_hair_bk_r_AIM}
    lookAtLocation=<-19.9959 -33.6015 8.0831>
    lookAtTarget {boneReference {boneName=spine_04}}
    lookAt_Axis {axis=<1 0>}
   }
   n52 = Look At - Bone: dyn_hair_bk_AIM(ComponentPose=@n51, Alpha=1) {
    alphaBoolBlend {blendOption=Linear}
    boneToModify {boneName=dyn_hair_bk_AIM}
    lookAtLocation=<-18.8903 -29.482>
    lookAtTarget {boneReference {boneName=spine_04}}
    lookAt_Axis {axis=<1 0>}
   }
   n53 = RigidBody(ComponentPose=@n52, OverrideWorldGravity=Get GravityOverride(), ComponentLinearAccScale=Get AccScaleOverride()) {
    bClampLinearTranslationLimitToRefPose=true
    bOverrideWorldGravity=true
    overridePhysicsAsset=&/M_MED_Ensemble_Snake_Vest_FaceAcc_Dyn_Hair
    simulationSpace=BaseBoneSpace
    alphaBoolBlend {blendOption=Linear}
    alphaScaleBiasClamp {interpSpeedDecreasing=4, interpSpeedIncreasing=6}
    baseBoneRef {boneName=spine_05}
    componentAppliedLinearAccClamp=<800 800 600>
    componentLinearAccScale=<0.5 0.18 0.1>
    componentLinearVelScale=<0.5 0.5 0.5>
    overrideWorldGravity {z=-875}
    simSpaceSettings {maxAngularAcceleration=200, maxAngularVelocity=5, maxLinearAcceleration=500, maxLinearVelocity=200}
   }
   n54 = "Transform (Modify) Bone - Bone: dyn_hair_side_r_MSTR"(ComponentPose=@n53, Alpha=Get Head_Rot_R(), Translation="0,0,0", Rotation="0,0,0", Scale="0.6,0.6,1") {
    rotationMode=BMM_Additive
    rotationSpace=BCS_BoneSpace
    scaleMode=BMM_Additive
    scaleSpace=BCS_BoneSpace
    translationMode=BMM_Additive
    translationSpace=BCS_BoneSpace
    alphaBoolBlend {blendOption=Linear}
    boneToModify {boneName=dyn_hair_side_r_MSTR}
    scale=<0.6 0.6>
   }
   Save cached pose 'noWearingPose'(Pose=Component To Local(ComponentPose=@n54)) {cacheName=noWearingPose}
   n55 = "Transform (Modify) Bone - Bone: dyn_hair_bk_MSTR"(
    ComponentPose=Local To Component(LocalPose=Use cached pose 'frontHairRBPose'())
    Alpha=@n6
    Translation="-5,0,2"
    Rotation="40,0,0"
    Scale="1,1,1"
   ) {
    rotationMode=BMM_Additive
    rotationSpace=BCS_BoneSpace
    translationMode=BMM_Additive
    translationSpace=BCS_BoneSpace
    alphaBoolBlend {blendOption=Linear}
    boneToModify {boneName=dyn_hair_bk_MSTR}
    rotation {pitch=40}
    translation {x=-5, z=2}
   }
   n56 = "Transform (Modify) Bone - Bone: dyn_hair_bk_r_MSTR"(ComponentPose=@n55, Alpha=@n6, Translation="-5,0,1", Rotation="40,0,0", Scale="1,1,1") {
    rotationMode=BMM_Additive
    rotationSpace=BCS_BoneSpace
    translationMode=BMM_Additive
    translationSpace=BCS_BoneSpace
    alphaBoolBlend {blendOption=Linear}
    boneToModify {boneName=dyn_hair_bk_r_MSTR}
    rotation {pitch=40}
    translation {x=-5, z=1}
   }
   n57 = "Transform (Modify) Bone - Bone: dyn_hair_bk_l_MSTR"(ComponentPose=@n56, Alpha=@n6, Translation="-5,0,1", Rotation="40,0,0", Scale="1,1,1") {
    rotationMode=BMM_Additive
    rotationSpace=BCS_BoneSpace
    translationMode=BMM_Additive
    translationSpace=BCS_BoneSpace
    alphaBoolBlend {blendOption=Linear}
    boneToModify {boneName=dyn_hair_bk_l_MSTR}
    rotation {pitch=40}
    translation {x=-5, z=1}
   }
   n58 = Get Head_Rot_R()
   n59 = "Transform (Modify) Bone - Bone: dyn_hair_bk_MSTR"(ComponentPose=@n57, Alpha=@n58, Translation="-2.818,2.169,-2.005", Rotation="20.000,-30,-0.038", Scale="1,1,1") {
    rotationMode=BMM_Additive
    rotationSpace=BCS_BoneSpace
    translationMode=BMM_Additive
    translationSpace=BCS_BoneSpace
    alphaBoolBlend {blendOption=Linear}
    boneToModify {boneName=dyn_hair_bk_MSTR}
    rotation=<rot 20 -30 -0.038>
    translation=<-2.818 2.169 -2.005>
   }
   n60 = "Transform (Modify) Bone - Bone: dyn_hair_bk_l_MSTR"(ComponentPose=@n59, Alpha=@n58, Translation="5,0,2", Rotation="30,0,0", Scale="1,1,1") {
    rotationMode=BMM_Additive
    rotationSpace=BCS_BoneSpace
    translationMode=BMM_Additive
    translationSpace=BCS_BoneSpace
    alphaBoolBlend {blendOption=Linear}
    boneToModify {boneName=dyn_hair_bk_l_MSTR}
    rotation {pitch=30}
    translation {x=5, z=2}
   }
   n61 = "Transform (Modify) Bone - Bone: dyn_hair_bk_r_MSTR"(ComponentPose=@n60, Alpha=@n58, Translation="0,0,-5", Rotation="10,-30,0", Scale="1,1,1") {
    rotationMode=BMM_Additive
    rotationSpace=BCS_BoneSpace
    translationMode=BMM_Additive
    translationSpace=BCS_BoneSpace
    alphaBoolBlend {blendOption=Linear}
    boneToModify {boneName=dyn_hair_bk_r_MSTR}
    rotation {pitch=10, yaw=-30}
    translation {z=-5}
   }
   n62 = Get Head_Rot_R()
   n63 = "Transform (Modify) Bone - Bone: dyn_hair_side_r_MSTR"(ComponentPose=@n61, Alpha=@n62, Translation="-3,3,0", Rotation="23.895,18.553,-7.001", Scale="1,1,1") {
    rotationMode=BMM_Additive
    rotationSpace=BCS_BoneSpace
    scaleMode=BMM_Additive
    scaleSpace=BCS_BoneSpace
    translationMode=BMM_Additive
    translationSpace=BCS_BoneSpace
    alphaBoolBlend {blendOption=Linear}
    boneToModify {boneName=dyn_hair_side_r_MSTR}
    rotation=<rot 23.895 18.5529 -7.001>
    translation=<-3 3>
   }
   n64 = "Transform (Modify) Bone - Bone: dyn_hair_side_r_AIM"(ComponentPose=@n63, Alpha=@n62, Translation="0,0,0", Rotation="0,0,0", Scale="1,1,1") {
    rotationMode=BMM_Additive
    rotationSpace=BCS_BoneSpace
    scaleMode=BMM_Additive
    scaleSpace=BCS_BoneSpace
    translationMode=BMM_Additive
    translationSpace=BCS_BoneSpace
    alphaBoolBlend {blendOption=Linear}
    boneToModify {boneName=dyn_hair_side_r_AIM}
   }
   n65 = "Transform (Modify) Bone - Bone: dyn_hair_side_l_MSTR"(ComponentPose=@n64, Alpha=@n62, Translation="-3,-3,0", Rotation="0.000,-45.000,-0.000", Scale="1,1,1") {
    rotationMode=BMM_Additive
    rotationSpace=BCS_BoneSpace
    scaleMode=BMM_Additive
    scaleSpace=BCS_BoneSpace
    translationMode=BMM_Additive
    translationSpace=BCS_BoneSpace
    alphaBoolBlend {blendOption=Linear}
    boneToModify {boneName=dyn_hair_side_l_MSTR}
    rotation=<rot 0 -44.9999 0>
    translation=<-3 -3>
   }
   n66 = Look At - Bone: dyn_hair_bk_l_AIM(ComponentPose=@n65, Alpha=1) {
    alphaBoolBlend {blendOption=Linear}
    boneToModify {boneName=dyn_hair_bk_l_AIM}
    lookAtLocation=<-20 -60 -8.3>
    lookAtTarget {boneReference {boneName=spine_02}}
    lookAt_Axis {axis=<1 0>}
   }
   n67 = Look At - Bone: dyn_hair_bk_r_AIM(ComponentPose=@n66, Alpha=1) {
    alphaBoolBlend {blendOption=Linear}
    boneToModify {boneName=dyn_hair_bk_r_AIM}
    lookAtLocation=<-20 -60 8.3>
    lookAtTarget {boneReference {boneName=spine_02}}
    lookAt_Axis {axis=<1 0>}
   }
   n68 = Look At - Bone: dyn_hair_bk_AIM(ComponentPose=@n67, Alpha=1) {
    alphaBoolBlend {blendOption=Linear}
    boneToModify {boneName=dyn_hair_bk_AIM}
    lookAtLocation=<-20.0035 -60.942 1.0901>
    lookAtTarget {boneReference {boneName=spine_02}}
    lookAt_Axis {axis=<1 0>}
   }
   n69 = Look At - Bone: dyn_hair_side_r_AIM(ComponentPose=@n68, Alpha=1) {
    alphaBoolBlend {blendOption=Linear}
    boneToModify {boneName=dyn_hair_side_r_AIM}
    lookAtLocation=<-30 -70 16>
    lookAtTarget {boneReference {boneName=spine_05}}
    lookAt_Axis {axis=<1 0>}
   }
   n70 = Look At - Bone: dyn_hair_side_l_AIM(ComponentPose=@n69, Alpha=1) {
    alphaBoolBlend {blendOption=Linear}
    boneToModify {boneName=dyn_hair_side_l_AIM}
    lookAtLocation=<-30 -72 -16>
    lookAtTarget {boneReference {boneName=spine_05}}
    lookAt_Axis {axis=<1 0>}
   }
   n71 = RigidBody(ComponentPose=@n70, OverrideWorldGravity=Get GravityOverride(), ComponentLinearAccScale=Get AccScaleOverride()) {
    bClampLinearTranslationLimitToRefPose=true
    bOverrideWorldGravity=true
    overridePhysicsAsset=/Game/Characters/Player/Male/Medium/Bodies/M_MED_Ensemble_Snake_Vest/Meshes/Parts/M_MED_Ensemble_Snake_Vest_FaceAcc_Dyn_Hair_Wearing_Cape
    simulationSpace=BaseBoneSpace
    alphaBoolBlend {blendOption=Linear}
    alphaScaleBiasClamp {interpSpeedDecreasing=4, interpSpeedIncreasing=6}
    baseBoneRef {boneName=neck_01}
    componentAppliedLinearAccClamp=<1000 1000 300>
    componentLinearAccScale=<0.3 0.3 0.2>
    componentLinearVelScale=<0.5 0.5 0.5>
    overrideWorldGravity {z=-875}
    simSpaceSettings {maxAngularAcceleration=200, maxAngularVelocity=5, maxLinearAcceleration=500, maxLinearVelocity=200}
   }
   n72 = "Transform (Modify) Bone - Bone: dyn_hair_side_r_MSTR"(ComponentPose=@n71, Alpha=Get Head_Rot_R(), Translation="0,0,0", Rotation="0,10,0", Scale="0.625,1.000,1.000") {
    rotationMode=BMM_Additive
    rotationSpace=BCS_BoneSpace
    scaleMode=BMM_Additive
    scaleSpace=BCS_BoneSpace
    translationMode=BMM_Additive
    translationSpace=BCS_BoneSpace
    alphaBoolBlend {blendOption=Linear}
    boneToModify {boneName=dyn_hair_side_r_MSTR}
    rotation {yaw=10}
    scale {x=0.625}
   }
   Save cached pose 'wearingCapePose'(Pose=Component To Local(ComponentPose=@n72)) {cacheName=wearingCapePose}
   n73 = "Constraint - dyn_hair_bk_r_AIM (PP)"!(ComponentPose=Local To Component(LocalPose=Use cached pose 'BasePose'()), Alpha=1, ConstraintWeights_0=0.2, ConstraintWeights_1:float=0.8) -> (Pose) {
    alphaBoolBlend {blendOption=Linear}
    boneToModify {boneName=dyn_hair_bk_r_AIM}
    constraintSetup [{targetBone {boneName=spine_05}}, {targetBone {boneName=head}}]
    constraintWeights [1,1]
   }
   n74 = "Constraint - dyn_hair_bk_l_AIM (PP)"!(ComponentPose=@n73, Alpha=1, ConstraintWeights_0=0.2, ConstraintWeights_1:float=0.8) -> (Pose) {
    alphaBoolBlend {blendOption=Linear}
    boneToModify {boneName=dyn_hair_bk_l_AIM}
    constraintSetup [{targetBone {boneName=spine_05}}, {targetBone {boneName=head}}]
    constraintWeights [1,1]
   }
   n75 = "Constraint - dyn_hair_bk_AIM (PP)"!(ComponentPose=@n74, Alpha=1, ConstraintWeights_0=0.2, ConstraintWeights_1:float=0.8) -> (Pose) {
    alphaBoolBlend {blendOption=Linear}
    boneToModify {boneName=dyn_hair_bk_AIM}
    constraintSetup [{targetBone {boneName=spine_05}}, {targetBone {boneName=head}}]
    constraintWeights [1,1]
   }
   n76 = "Constraint - dyn_hair_bk_02 (PP)"!(ComponentPose=@n75, Alpha=1, ConstraintWeights_0=0.9, ConstraintWeights_1:float=0.1) -> (Pose) {
    alphaBoolBlend {blendOption=Linear}
    boneToModify {boneName=dyn_hair_bk_02}
    constraintSetup [{targetBone {boneName=spine_05}}, {targetBone {boneName=head}}]
    constraintWeights [1,1]
   }
   n77 = "Constraint - dyn_hair_bk_r_02 (PPP)"!(ComponentPose=@n76, Alpha=1, ConstraintWeights_0=0.75, ConstraintWeights_1:float=0.15, ConstraintWeights_2:float=0.1) -> (Pose) {
    alphaBoolBlend {blendOption=Linear}
    boneToModify {boneName=dyn_hair_bk_r_02}
    constraintSetup [{targetBone {boneName=spine_05}}, {targetBone {boneName=clavicle_r}}, {targetBone {boneName=head}}]
    constraintWeights [1,1,1]
   }
   n78 = "Constraint - dyn_hair_bk_l_02 (PPP)"!(ComponentPose=@n77, Alpha=1, ConstraintWeights_0=0.75, ConstraintWeights_1:float=0.15, ConstraintWeights_2:float=0.1) -> (Pose) {
    alphaBoolBlend {blendOption=Linear}
    boneToModify {boneName=dyn_hair_bk_l_02}
    constraintSetup [{targetBone {boneName=spine_05}}, {targetBone {boneName=clavicle_l}}, {targetBone {boneName=head}}]
    constraintWeights [1,1,1]
   }
   n79 = "Constraint - dyn_hair_side_r_AIM (PPP)"!(ComponentPose=@n78, Alpha=1, ConstraintWeights_0=0.75, ConstraintWeights_1:float=0.15, ConstraintWeights_2:float=0.1) -> (Pose) {
    alphaBoolBlend {blendOption=Linear}
    boneToModify {boneName=dyn_hair_side_r_AIM}
    constraintSetup [{targetBone {boneName=spine_05}}, {targetBone {boneName=clavicle_r}}, {targetBone {boneName=head}}]
    constraintWeights [1,1,1]
   }
   n80 = "Constraint - dyn_hair_side_l_AIM (PPP)"!(ComponentPose=@n79, Alpha=1, ConstraintWeights_0=0.75, ConstraintWeights_1:float=0.15, ConstraintWeights_2:float=0.1) -> (Pose) {
    alphaBoolBlend {blendOption=Linear}
    boneToModify {boneName=dyn_hair_side_l_AIM}
    constraintSetup [{targetBone {boneName=spine_05}}, {targetBone {boneName=clavicle_l}}, {targetBone {boneName=head}}]
    constraintWeights [1,1,1]
   }
   n81 = "Constraint - dyn_hair_bk_03 (P)"(ComponentPose=@n80, Alpha=1, ConstraintWeights_0=1) {
    alphaBoolBlend {blendOption=Linear}
    boneToModify {boneName=dyn_hair_bk_03}
    constraintSetup [{targetBone {boneName=spine_05}}]
    constraintWeights [1]
   }
   n82 = "Constraint - dyn_hair_bk_r_03 (P)"(ComponentPose=@n81, Alpha=1, ConstraintWeights_0=1) {
    alphaBoolBlend {blendOption=Linear}
    boneToModify {boneName=dyn_hair_bk_r_03}
    constraintSetup [{targetBone {boneName=spine_05}}]
    constraintWeights [1]
   }
   n83 = "Constraint - dyn_hair_bk_l_03 (P)"(ComponentPose=@n82, Alpha=1, ConstraintWeights_0=1) {
    alphaBoolBlend {blendOption=Linear}
    boneToModify {boneName=dyn_hair_bk_l_03}
    constraintSetup [{targetBone {boneName=spine_05}}]
    constraintWeights [1]
   }
   n84 = "Constraint - dyn_hair_bk_l_04 (PPP)"!(ComponentPose=@n83, Alpha=1, ConstraintWeights_0=0.4, ConstraintWeights_1:float=0.35, ConstraintWeights_2:float=0.25) -> (Pose) {
    alphaBoolBlend {blendOption=Linear}
    boneToModify {boneName=dyn_hair_bk_l_04}
    constraintSetup [{targetBone {boneName=spine_03}}, {targetBone {boneName=spine_04}}, {targetBone {boneName=spine_05}}]
    constraintWeights [1,1,1]
   }
   n85 = "Constraint - dyn_hair_bk_04 (PPP)"!(ComponentPose=@n84, Alpha=1, ConstraintWeights_0=0.4, ConstraintWeights_1:float=0.35, ConstraintWeights_2:float=0.25) -> (Pose) {
    alphaBoolBlend {blendOption=Linear}
    boneToModify {boneName=dyn_hair_bk_04}
    constraintSetup [{targetBone {boneName=spine_03}}, {targetBone {boneName=spine_04}}, {targetBone {boneName=spine_05}}]
    constraintWeights [1,1,1]
   }
   n86 = "Constraint - dyn_hair_bk_r_04 (PPP)"!(ComponentPose=@n85, Alpha=1, ConstraintWeights_0=0.4, ConstraintWeights_1:float=0.35, ConstraintWeights_2:float=0.25) -> (Pose) {
    alphaBoolBlend {blendOption=Linear}
    boneToModify {boneName=dyn_hair_bk_r_04}
    constraintSetup [{targetBone {boneName=spine_03}}, {targetBone {boneName=spine_04}}, {targetBone {boneName=spine_05}}]
    constraintWeights [1,1,1]
   }
   n87 = "Constraint - dyn_hair_bk_05 (PPP)"!(ComponentPose=@n86, Alpha=1, ConstraintWeights_0=0.8, ConstraintWeights_1:float=0.2, ConstraintWeights_2:float=0) -> (Pose) {
    alphaBoolBlend {blendOption=Linear}
    boneToModify {boneName=dyn_hair_bk_05}
    constraintSetup [{targetBone {boneName=spine_03}}, {targetBone {boneName=spine_04}}, {targetBone {boneName=spine_05}}]
    constraintWeights [1,1,1]
   }
   n88 = "Constraint - dyn_hair_ft_l_02 (PP)"!(ComponentPose=@n87, Alpha=1, ConstraintWeights_0=0.7, ConstraintWeights_1:float=0.3) -> (Pose) {
    alphaBoolBlend {blendOption=Linear}
    boneToModify {boneName=dyn_hair_ft_l_02}
    constraintSetup [{targetBone {boneName=spine_05}}, {targetBone {boneName=head}}]
    constraintWeights [1,1]
   }
   n89 = "Constraint - dyn_hair_ft_r_02 (PP)"!(ComponentPose=@n88, Alpha=1, ConstraintWeights_0=0.7, ConstraintWeights_1:float=0.3) -> (Pose) {
    alphaBoolBlend {blendOption=Linear}
    boneToModify {boneName=dyn_hair_ft_r_02}
    constraintSetup [{targetBone {boneName=spine_05}}, {targetBone {boneName=head}}]
    constraintWeights [1,1]
   }
   n90 = "Constraint - dyn_hair_ft_l_03 (P)"(ComponentPose=@n89, Alpha=1, ConstraintWeights_0=1) {
    alphaBoolBlend {blendOption=Linear}
    boneToModify {boneName=dyn_hair_ft_l_03}
    constraintSetup [{targetBone {boneName=spine_05}}]
    constraintWeights [1]
   }
   n91 = "Constraint - dyn_hair_ft_r_03 (P)"(ComponentPose=@n90, Alpha=1, ConstraintWeights_0=1) {
    alphaBoolBlend {blendOption=Linear}
    boneToModify {boneName=dyn_hair_ft_r_03}
    constraintSetup [{targetBone {boneName=spine_05}}]
    constraintWeights [1]
   }
   Save cached pose 'noRigidBodyPose'(Pose=Component To Local(ComponentPose=@n91)) {cacheName=noRigidBodyPose}
   n92 = "Constraint - dyn_hair_bk_r_AIM (PP)"(ComponentPose=Local To Component(LocalPose=Use cached pose 'frontHairRBPose'()), Alpha=1, ConstraintWeights_0=0.2, ConstraintWeights_1=0.8) {
    alphaBoolBlend {blendOption=Linear}
    boneToModify {boneName=dyn_hair_bk_r_AIM}
    constraintSetup [{targetBone {boneName=spine_05}}, {targetBone {boneName=head}}]
    constraintWeights [1,1]
   }
   n93 = "Constraint - dyn_hair_bk_l_AIM (PP)"(ComponentPose=@n92, Alpha=1, ConstraintWeights_0=0.2, ConstraintWeights_1=0.8) {
    alphaBoolBlend {blendOption=Linear}
    boneToModify {boneName=dyn_hair_bk_l_AIM}
    constraintSetup [{targetBone {boneName=spine_05}}, {targetBone {boneName=head}}]
    constraintWeights [1,1]
   }
   n94 = "Constraint - dyn_hair_bk_AIM (PP)"(ComponentPose=@n93, Alpha=1, ConstraintWeights_0=0.2, ConstraintWeights_1=0.8) {
    alphaBoolBlend {blendOption=Linear}
    boneToModify {boneName=dyn_hair_bk_AIM}
    constraintSetup [{targetBone {boneName=spine_05}}, {targetBone {boneName=head}}]
    constraintWeights [1,1]
   }
   n95 = "Constraint - dyn_hair_bk_02 (PP)"(ComponentPose=@n94, Alpha=1, ConstraintWeights_0=0.9, ConstraintWeights_1=0.1) {
    alphaBoolBlend {blendOption=Linear}
    boneToModify {boneName=dyn_hair_bk_02}
    constraintSetup [{targetBone {boneName=spine_05}}, {targetBone {boneName=head}}]
    constraintWeights [1,1]
   }
   n96 = "Constraint - dyn_hair_bk_r_02 (PPP)"(ComponentPose=@n95, Alpha=1, ConstraintWeights_0=0.75, ConstraintWeights_1=0.15, ConstraintWeights_2=0.1) {
    alphaBoolBlend {blendOption=Linear}
    boneToModify {boneName=dyn_hair_bk_r_02}
    constraintSetup [{targetBone {boneName=spine_05}}, {targetBone {boneName=clavicle_r}}, {targetBone {boneName=head}}]
    constraintWeights [1,1,1]
   }
   n97 = "Constraint - dyn_hair_bk_l_02 (PPP)"(ComponentPose=@n96, Alpha=1, ConstraintWeights_0=0.75, ConstraintWeights_1=0.15, ConstraintWeights_2=0.1) {
    alphaBoolBlend {blendOption=Linear}
    boneToModify {boneName=dyn_hair_bk_l_02}
    constraintSetup [{targetBone {boneName=spine_05}}, {targetBone {boneName=clavicle_l}}, {targetBone {boneName=head}}]
    constraintWeights [1,1,1]
   }
   n98 = "Constraint - dyn_hair_side_r_AIM (PPP)"(ComponentPose=@n97, Alpha=1, ConstraintWeights_0=0.75, ConstraintWeights_1=0.15, ConstraintWeights_2=0.1) {
    alphaBoolBlend {blendOption=Linear}
    boneToModify {boneName=dyn_hair_side_r_AIM}
    constraintSetup [{targetBone {boneName=spine_05}}, {targetBone {boneName=clavicle_r}}, {targetBone {boneName=head}}]
    constraintWeights [1,1,1]
   }
   n99 = "Constraint - dyn_hair_side_l_AIM (PPP)"(ComponentPose=@n98, Alpha=1, ConstraintWeights_0=0.75, ConstraintWeights_1=0.15, ConstraintWeights_2=0.1) {
    alphaBoolBlend {blendOption=Linear}
    boneToModify {boneName=dyn_hair_side_l_AIM}
    constraintSetup [{targetBone {boneName=spine_05}}, {targetBone {boneName=clavicle_l}}, {targetBone {boneName=head}}]
    constraintWeights [1,1,1]
   }
   n100 = RigidBody!(ComponentPose=@n99, Alpha, OverrideWorldGravity="0,0,-700", ExternalForce) -> (Pose) {
    bClampLinearTranslationLimitToRefPose=true
    bOverrideWorldGravity=true
    overridePhysicsAsset=&/M_MED_Ensemble_Snake_Vest_FaceAcc_Dyn_Hair_Wearing_Backbling
    simulationSpace=BaseBoneSpace
    alphaBoolBlend {blendOption=Linear}
    alphaScaleBiasClamp {interpSpeedDecreasing=4, interpSpeedIncreasing=6}
    baseBoneRef {boneName=spine_05}
    componentAppliedLinearAccClamp=<100 100 600>
    componentLinearAccScale=<0.1 0.1 0.15>
    componentLinearVelScale=<0.5 0.5 0.5>
    overrideWorldGravity {z=-875}
    simSpaceSettings {maxAngularAcceleration=200, maxAngularVelocity=5, maxLinearAcceleration=500, maxLinearVelocity=200}
   }
   Save cached pose 'wearingBackblingPose'(Pose=Component To Local(ComponentPose=@n100)) {cacheName=wearingBackblingPose}
   Get GravityOverride()
  ]}
 EventGraph {nodes [
   Event BlueprintUpdateAnimation()
   > n1 = Sequence()
   n2 = GetOwnerSkeletalMeshComponent()
   n3 = BreakRotator(InRot=BreakTransform(InTransform=GetDeltaTransformFromRefPose(self=@n2, BoneName=head, BaseName=spine_05)).Rotation)
   > Set Head_Rot_R(Head_Rot_R:double=MapRangeClamped(Value=@n3.Pitch, InRangeA=25, InRangeB=52)) -> (Output_Get:double)
   > Set Head_Rot_Bk(Head_Rot_Bk:double=MapRangeClamped(Value=@n3.Yaw, InRangeA=-10, InRangeB=-55)) -> (Output_Get:double)
   > Set Head_Rot_Dn(Head_Rot_Dn:double=MapRangeClamped(Value=@n3.Yaw, InRangeA=0, InRangeB=30)) -> (Output_Get:double)
   > Set Arm_Up_R(Arm_Up_R:double=GetAlphaFromReferenceTransformDelta(SkeletalMeshComponent=@n2, BoneName=upperarm_r, BaseName=spine_05, InputRange=(X=30,Y=123), TwistAxis="1,0,0")) -> (Output_Get:double)
   n4 = Gravity Override State Function DEPRECATED(
    execute=@n1.then_1
    CharacterPartAnimInstance=Self-Reference()
    GameTimeInSeconds=GetGameTimeInSeconds()
    GravityOverrideParams=Get Gravity Override Params()
   )
   > Set GravityOverride(GravityOverride=@n4)
   > n5 = Sequence()
   > n6 = Branch(Condition=OR Boolean(A=Get bIsSkyDiving(), B=Get bIsCrouching()))
   > Set AccScaleOverride(AccScaleOverride:Vector="0.05,0.05,0.1") -> (Output_Get:Vector)
   Set AccScaleOverride(execute=@n6.else, AccScaleOverride="0.3,0.18,0.15")
   n7 = Branch(execute=@n5.then_1, Condition=Get isWearBackbling())
   > n8 = Branch(Condition=Get isWearCape())
   > Set AccScaleOverride(AccScaleOverride="0.3,0.3,0.15")
   Set AccScaleOverride(execute=@n8.else, AccScaleOverride="0.05,0.05,0.1")
   Set AccScaleOverride(execute=@n7.else, AccScaleOverride="0.3,0.18,0.15")
   Event OnCharacterCustomizationLoaded()
   > n9 = Sequence()
   > Delay(Duration=0.1)
   > ResetDynamics(InTeleportType=TeleportPhysics)
   > ResetDynamics(InTeleportType=ResetPhysics)
   Do Once(execute=@n9.then_1)
   n10 = CheckForBackbling()
   > Set isWearBackbling(isWearBackbling:bool=@n10) -> (Output_Get:bool)
   > Set isWearCape(isWearCape:bool=@n10.bIsBackblingLong) -> (Output_Get:bool)
   Event BlueprintInitializeAnimation()
   > Set Body_Mesh(Body_Mesh=GetOwnerSkeletalMeshComponent())
   TryGetPawnOwner()
  ]}
]
variables [
 AccScaleOverride {type=Vector}
 Arm_Up_R {type=number}
 Body_Mesh {type=SkeletalMeshComponent}
 Gravity Override Params {
  type=GravityOverrideParamsStruct
  default {
   dBNO {gravityOverride=<-700 200 0>}
   dBNO_Carried {gravityOverride=<-700 300 -300>}
   emoteOrMelee {gravityOverride=<0 0 -700>}
   front_End {gravityOverride=<0 0 -700>}
   onGround_Falling {gravityOverride=<0 0 -700>}
   onGround_Moving {gravityOverride=<0 0 -700>}
   onGround_Standing {gravityOverride=<0 0 -700>}
   skydive_DivingDown {gravityOverride=<-1500 600 0>, windAmplitude=<50 50 100>}
   skydive_FloatingDown {gravityOverride=<0 -100 -500>, windAmplitude=<50 -100 50>, windOffset=<0 0 50>}
   skydive_GoingUp {gravityOverride=<0 -100 -300>, windAmplitude=<50 -200 100>}
   skydive_Parachuting {gravityOverride=<0 -500 -700>, windAmplitude=<100 -150 200>}
   swimming {gravityOverride=<-500 200 0>}
   ziplining {gravityOverride=<0 0 -700>}
  }
 }
 GravityOverride {type=Vector}
 Head_Rot_Bk {type=number}
 Head_Rot_Dn {type=number}
 Head_Rot_R {type=number}
 isWearBackbling {type=boolean}
 isWearCape {type=boolean}
]
